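# Editor utility widget for finding "invalid" DFX_MaterialComponents on level actors.
# Find button casts each actor to BuildingSMActor; the Cast Failed branch reads components from the cast result, not the loop actor.
# Valid components are added to the ComponentsArray map with their owners, displayed in a DetailsView bound to the widget on PreConstruct.
# The "Destroy invalid Components" button has no logic.
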EditorUtilityWidgetBlueprint /Game/ArtTools/Tech_Art/DirectFX/DFX_ComponentCleaner
graphs [EventGraph {nodes [
   "On Clicked (EditorUtilityButton_65)"() {is=title K2Node_ComponentBoundEvent}
   > n1 = GetAllActorsOfClass!(execute, WorldContextObject:Object Reference, ActorClass:Actor Class Reference=&Actor) -> (then, OutActors:Array of Actor Object References)
   > n2 = For Each Loop(Array:Array of Actor Object References=@n1) -> (Array Element:Actor Object Reference)
   > n3 = Cast To BuildingSMActor!(execute, Object=@n2) -> (then, CastFailed, AsBuilding Static Mesh:Building Static Mesh Object Reference) {is=title K2Node_DynamicCast}
   n4 = For Each Loop(
    Exec=@n3.CastFailed
    Array:Array of DFX Material Component Object References=Get Components By Class!(
     self:Actor Object Reference=@n3
     ComponentClass:Actor Component Class Reference=&/DFX_MaterialComponent.DFX_MaterialComponent_C
    ) -> (ReturnValue:Array of DFX Material Component Object References)
   ) -> (Array Element:DFX Material Component Object Reference)
   > Is Valid(InputObject=@n4)
   > Add!(
    execute
    TargetMap:Map of DFX Material Component Object References to Actor Object References=Get ComponentsArray() -> (ComponentsArray:Map of DFX Material Component Object References to Actor Object References)
    Key:"DFX Material Component Object Reference (by ref)"=@n4
    Value:"Actor Object Reference (by ref)"=GetOwner(self=@n4)
   ) -> (then) {is=title K2Node_CallFunction}
   Event PreConstruct()
   > SetObject(self=Get DetailsView_66() -> (DetailsView_66:Details View Object Reference), NewObject=Self-Reference())
  ]}]
variables [ComponentsArray {default {}}]
widgets {
 info {rootWidgetClass=&VerticalBox, widgetCount=6}
 widgets [
  VerticalBox_30 {widgetClassPath=&VerticalBox, widget=VerticalBox_30}
  EditorUtilityButton_65 {widgetClassPath=&EditorUtilityButton, parent=VerticalBox_30, widget=EditorUtilityButton_65, slot {class=VerticalBoxSlot}}
  TextBlock_25 {
   widgetClassPath=&TextBlock
   parent=EditorUtilityButton_65
   widget=TextBlock_25
   properties {text=Find invalid Components}
   slot {class=ButtonSlot}
  }
  EditorUtilityButton {widgetClassPath=&EditorUtilityButton, parent=VerticalBox_30, widget=EditorUtilityButton, slot {class=VerticalBoxSlot}}
  TextBlock {
   widgetClassPath=&TextBlock
   parent=EditorUtilityButton
   widget=TextBlock
   properties {text=Destroy invalid Components}
   slot {class=ButtonSlot}
  }
  DetailsView_66 {
   widgetClassPath=&DetailsView
   parent=VerticalBox_30
   widget=DetailsView_66
   properties {propertiesToShow [ComponentsArray]}
   slot {class=VerticalBoxSlot}
  }
 ]
}
