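; DFX material component that puts a projected DirectFX dynamic material on its owner's building or static mesh.
; Update DFX, run at BeginPlay, finds the mesh, refreshes the MID when its parent differs and pushes DFX struct parameters into it.
; For slot 0 it copies the building's textures into the MID and enables custom-material flags.
; Destroy DFX clears those flags and removes the component; ResetMaterials restores original materials.
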(event EventBeginPlay
  (CallFunction|UpdateDFX))

(event Custom|Init
  (CallFunction|UpdateBuildingCustomMaterialParameter))

(event Custom|Update%20DFX
  (bind _returnvalue (Components|GetOwner))
  (bind _returnvalue_2 (Actor|GetComponentbyClass _returnvalue "/Script/Engine.StaticMeshComponent"))
  (bind _mid (Variables|Default|GetMID))
  (bind _mid_2 (Variables|Default|GetMID))
  (CallFunction|UpdateBuildingCustomMaterialParameter)
  (Utilities|Casting|CastToDFX_Placement _returnvalue
    (:CastFailed
      (bind _asbuilding_static_mesh (Utilities|Casting|CastToBuildingSMActor _returnvalue)
        (:then
          (Variables|Default|SetStaticMeshComponent (Variables|/Script/FortniteGame.BuildingSMActor|GetStaticMeshComponent _asbuilding_static_mesh)))
        (:CastFailed
          (bind _asbuilding_static_mesh__static_geo (Utilities|Casting|CastToBuildingStaticGeo _returnvalue)
            (:then
              (Variables|Default|SetStaticGeoMat (Rendering|Material|GetMaterial (Variables|/Script/Engine.StaticMeshActor|GetStaticMeshComponent _asbuilding_static_mesh__static_geo) 0)))
            (:CastFailed
              _))
          (Utilities|IsValid _returnvalue_2
            (:Is%20Valid
              (Variables|Default|SetStaticMeshComponent _returnvalue_2))
            (:Is%20Not%20Valid
              (Development|PrintString (Utilities|String|Append "No Static mesh found in this actor ->  " (Utilities|GetDisplayName _returnvalue)) true true "(R=1.000000,G=0.000000,B=0.043681,A=1.000000)")))))
      (Utilities|IsValid (Variables|Default|GetParentMaterial)
        (:Is%20Valid
          (Utilities|IsValid _mid
            (:Is%20Valid
              (if (!= (Variables|/Script/Engine.MaterialInstance|GetParent _mid) (Variables|Default|GetParentMaterial))
                (bind _returnvalue_3 (Rendering|Material|CreateDynamicMaterialInstance (Variables|Default|GetParentMaterial) "None" "None" :!class /Script/Engine.KismetMaterialLibrary))
                (Variables|Default|SetMID _returnvalue_3))
              (Utilities|IsValid (Variables|Default|GetMID)
                (:Is%20Valid
                  (CallFunction|SetMaterialParams _mid_2)
                  (Rendering|Material|SetMaterial (Variables|Default|GetStaticMeshComponent) (Math|Integer|Min(Integer) (- (Rendering|Material|GetNumMaterials (Variables|Default|GetStaticMeshComponent)) 1) (Variables|Default|GetMaterialSlot)) _mid_2))))
            (:Is%20Not%20Valid
              _)))))))

(event Custom|Destroy%20DFX
  (bind _returnvalue (Components|GetOwner))
  (Utilities|FlowControl|Sequence
    (:then_0
      (bind _asbuilding_static_mesh (Utilities|Casting|CastToBuildingSMActor _returnvalue)
        (:then
          (Variables|/Script/FortniteGame.BuildingSMActor|SetbAllowCustomMaterial false _asbuilding_static_mesh)
          (Variables|/Script/FortniteGame.BuildingSMActor|SetbConsiderOverrideMaterialsOnAnimationCleanup false _asbuilding_static_mesh)
          (Components|DestroyComponent))
        (:CastFailed
          (bind _asbuilding_static_mesh__static_geo (Utilities|Casting|CastToBuildingStaticGeo _returnvalue))
          (Variables|/Script/FortniteGame.BuildingStaticGeo|SetbAllowCustomMaterial false _asbuilding_static_mesh__static_geo))))
    (:then_1
      (Components|DestroyComponent))))

(fn ApplyTextureData
  (params (Material :type Material%20Instance%20Dynamic%20Object%20Reference))
  (bind _returnvalue (Components|GetOwner))
  (if (== (Variables|Default|GetMaterialSlot) 0)
    (bind _asbuilding_static_mesh (Utilities|Casting|CastToBuildingSMActor _returnvalue)
      (:then
        (bind _returnvalue_2 (Class|BuildingStaticMesh|ReadTextureData _asbuilding_static_mesh))
        (LocalVariables|SetLTextures _returnvalue_2))
      (:CastFailed
        (bind _asbuilding_static_mesh__static_geo (Utilities|Casting|CastToBuildingStaticGeo _returnvalue))
        (bind _returnvalue_3 (Class|BuildingStaticMesh%28StaticGeo%29|ReadTextureData _asbuilding_static_mesh__static_geo))
        (LocalVariables|SetLTextures _returnvalue_3)))
    (for (_array_element _array_index) (LocalVariables|GetLTextures)
      (Utilities|IsValid _array_element
        (:Is%20Valid
          (Rendering|Material|SetTextureParameterValue Material (Utilities|Array|Get(aref) (Utilities|Array|MakeArray "Diffuse" "Normals" "SpecularMasks" "Diffuse_Texture_2" "Normals_Texture_2" "SpecularMasks_2" "Diffuse_Texture_3" "Normals_Texture_3" "SpecularMasks_3" "Diffuse_Texture_4" "Normals_Texture_4" "SpecularMasks_4") _array_index) _array_element))))))

(fn ConvertAlphaSourceToFloat)

(fn ResetMaterials
  (bind _returnvalue (Components|GetOwner))
  (bind _returnvalue_2 (Actor|GetComponentbyClass _returnvalue "/Script/Engine.StaticMeshComponent"))
  (Utilities|Casting|CastToBuildingStaticGeo _returnvalue
    (:then
      (Rendering|Material|SetMaterial _returnvalue_2 0 (Variables|Default|GetStaticGeoMat)))
    (:CastFailed
      (for (_array_element _array_index) (Variables|/Script/Engine.StaticMesh|GetStaticMaterials (Variables|/Script/Engine.StaticMeshComponent|GetStaticMesh _returnvalue_2))
        (Rendering|Material|SetMaterial _returnvalue_2 _array_index (Utilities|Struct|BreakStaticMaterial _array_element))))))

(fn SetMaterialParams
  (params (Material :type Material%20Instance%20Dynamic%20Object%20Reference))
  (bind (_dfx_materialstruct_texture_8_367f09c249db49b23753f695eafedec6 _dfx_materialstruct_matrix_x_9_863869604343f3cdff972aad5f2f3e0c _dfx_materialstruct_matrix_y_10_3a617e204c378484bb53c39a363d47f7 _dfx_materialstruct_matrix_z_11_7ec280b140c8e9a69aa9e69aa8f5880c _dfx_materialstruct_forwardvector_13_cd800e244085cc820c672ebe67112b8d _dfx_materialstruct_tintcolor_39_0e2067244504b0a20a9527ad96e915bc _dfx_materialstruct_panningspeed_42_4402686b48249213219afd9c82dadf67 _dfx_materialstruct_angleblend_18_66e99de84fca3617b0730793456181f0 _dfx_materialstruct_angleblendhardness_19_1e0491b342de0831ef95a0b645357c86 _dfx_materialstruct_perspective_22_eb3e14f249e79424bed54799cb192106 _dfx_materialstruct_fixedtexture_25_7973695749f65934722aa48dac6090ec _dfx_materialstruct_tiling_27_6dd3d5834da01bba73d945acf2c6f0e5 _dfx_materialstruct_intensity_30_87a506f949ac6fa6df2b40944de9ec5b _dfx_materialstruct_falloff_35_b4677d51444bc62af1e9d88ca9f0283c _dfx_materialstruct_falloffradius_36_6bc0bcf84f1463d127628f923fac0a08 _dfx_materialstruct_opacity_45_3595c8d94d378c0cc81cb5bf5b81d086 _dfx_materialstruct_nsrtexture_60_de05dd7643c72492beb0c3818128c9f5 _dfx_materialstruct_specular_59_5bcef35b439119b121d93c91c7637e91 _dfx_materialstruct_roughness_58_00537e4e402b49203edd97a491cbbe07 _dfx_materialstruct_normal_57_85340b55404efc7bf7508d98fe6923bf _dfx_materialstruct_emissive_56_df7b0427485a130162a6fcaa240f43d3 _dfx_materialstruct_rightvector_64_a7c9ea4d4c5683aefb0407a6e07adc01 _dfx_materialstruct_metallic_67_a381fbce442851caa373fbb66367d0d3 _dfx_materialstruct_specularoverride_69_65f02ab84bc77401d89db2bcd72fefcb _dfx_materialstruct_tilingy_72_0ae263f646f6e19e43a858bdfc838a02 _dfx_materialstruct_tilingx_75_ba80eaeb430531dc05a2948ef05cace9) (Utilities|Struct|BreakDFXMaterialParamsStruct (Variables|GetDFX_MaterialStruct_Texture_8_367F09C249DB49B23753F695EAFEDEC6)))
  (bind (_dfx_materialstruct_texture_8_367f09c249db49b23753f695eafedec6_2 _dfx_materialstruct_matrix_x_9_863869604343f3cdff972aad5f2f3e0c_2 _dfx_materialstruct_matrix_y_10_3a617e204c378484bb53c39a363d47f7_2 _dfx_materialstruct_matrix_z_11_7ec280b140c8e9a69aa9e69aa8f5880c_2 _dfx_materialstruct_forwardvector_13_cd800e244085cc820c672ebe67112b8d_2 _dfx_materialstruct_tintcolor_39_0e2067244504b0a20a9527ad96e915bc_2 _dfx_materialstruct_panningspeed_42_4402686b48249213219afd9c82dadf67_2 _dfx_materialstruct_angleblend_18_66e99de84fca3617b0730793456181f0_2 _dfx_materialstruct_angleblendhardness_19_1e0491b342de0831ef95a0b645357c86_2 _dfx_materialstruct_perspective_22_eb3e14f249e79424bed54799cb192106_2 _dfx_materialstruct_fixedtexture_25_7973695749f65934722aa48dac6090ec_2 _dfx_materialstruct_tiling_27_6dd3d5834da01bba73d945acf2c6f0e5_2 _dfx_materialstruct_intensity_30_87a506f949ac6fa6df2b40944de9ec5b_2 _dfx_materialstruct_falloff_35_b4677d51444bc62af1e9d88ca9f0283c_2 _dfx_materialstruct_falloffradius_36_6bc0bcf84f1463d127628f923fac0a08_2 _dfx_materialstruct_opacity_45_3595c8d94d378c0cc81cb5bf5b81d086_2 _dfx_materialstruct_nsrtexture_60_de05dd7643c72492beb0c3818128c9f5_2) (Utilities|Struct|BreakDFXMaterialParamsStruct (Variables|GetDFX_MaterialStruct_Texture_8_367F09C249DB49B23753F695EAFEDEC6)))
  (Rendering|Material|SetTextureParameterValue Material "DFX_Texture" _dfx_materialstruct_texture_8_367f09c249db49b23753f695eafedec6_2)
  (Rendering|Material|SetTextureParameterValue Material "DFX_NMRE" _dfx_materialstruct_nsrtexture_60_de05dd7643c72492beb0c3818128c9f5_2)
  (CallFunction|SetMaterialVectors Material (Utilities|Array|MakeArray _dfx_materialstruct_matrix_x_9_863869604343f3cdff972aad5f2f3e0c_2 _dfx_materialstruct_matrix_y_10_3a617e204c378484bb53c39a363d47f7_2 _dfx_materialstruct_matrix_z_11_7ec280b140c8e9a69aa9e69aa8f5880c_2 _dfx_materialstruct_forwardvector_13_cd800e244085cc820c672ebe67112b8d_2 _dfx_materialstruct_tintcolor_39_0e2067244504b0a20a9527ad96e915bc_2 _dfx_materialstruct_panningspeed_42_4402686b48249213219afd9c82dadf67_2 (Utilities|Select :Index (Variables|Default|GetAlphaSource) :NewEnumerator0 "(R=1.000000,G=0.000000,B=0.000000,A=0.000000)" :NewEnumerator1 "(R=0.000000,G=1.000000,B=0.000000,A=0.000000)" :NewEnumerator2 "(R=0.000000,G=0.000000,B=1.000000,A=0.000000)" :NewEnumerator3 "(R=0.000000,G=0.000000,B=0.000000,A=1.000000)")) (Utilities|Array|MakeArray "DFX_Matrix_X" "DFX_Matrix_Y" "DFX_Matrix_Z" "DFX_ForwardVector" "DFX_TintColor" "DFX_PanningSpeed" "AlphaSource"))
  (CallFunction|SetMaterialScalars Material (Utilities|Array|MakeArray _dfx_materialstruct_angleblend_18_66e99de84fca3617b0730793456181f0 _dfx_materialstruct_angleblendhardness_19_1e0491b342de0831ef95a0b645357c86 _dfx_materialstruct_perspective_22_eb3e14f249e79424bed54799cb192106 _dfx_materialstruct_fixedtexture_25_7973695749f65934722aa48dac6090ec _dfx_materialstruct_tiling_27_6dd3d5834da01bba73d945acf2c6f0e5 _dfx_materialstruct_intensity_30_87a506f949ac6fa6df2b40944de9ec5b _dfx_materialstruct_falloff_35_b4677d51444bc62af1e9d88ca9f0283c _dfx_materialstruct_falloffradius_36_6bc0bcf84f1463d127628f923fac0a08 _dfx_materialstruct_opacity_45_3595c8d94d378c0cc81cb5bf5b81d086 _dfx_materialstruct_tilingx_75_ba80eaeb430531dc05a2948ef05cace9 _dfx_materialstruct_tilingy_72_0ae263f646f6e19e43a858bdfc838a02 (Variables|Default|GetAlphaContrast) (Math|Conversions|ToFloat(Boolean) (Variables|Default|GetInvertAlpha))) (Utilities|Array|MakeArray "DFX_AngleBlend" "DFX_AngleBlendHardness" "DFX_Perspective" "DFX_FixedTexture" "DFX_Tiling" "DFX_Intensity" "DFX_Falloff" "DFX_FalloffRadius" "DFX_Opacity" "DFX_Tiling_X" "DFX_Tiling_Y" "AlphaContrast" "InvertAlpha"))
  (if (not (Variables|Default|GetForceNSRDisabled))
    (CallFunction|SetMaterialScalars Material (Utilities|Array|MakeArray _dfx_materialstruct_metallic_67_a381fbce442851caa373fbb66367d0d3 _dfx_materialstruct_roughness_58_00537e4e402b49203edd97a491cbbe07 _dfx_materialstruct_normal_57_85340b55404efc7bf7508d98fe6923bf _dfx_materialstruct_emissive_56_df7b0427485a130162a6fcaa240f43d3 _dfx_materialstruct_specularoverride_69_65f02ab84bc77401d89db2bcd72fefcb) (Utilities|Array|MakeArray "DFX_Metallic" "DFX_Roughness" "DFX_Normal" "DFX_Emissive" "DFX_SpecularOverride"))
    (CallFunction|SetMaterialVectors Material (Utilities|Array|MakeArray _dfx_materialstruct_rightvector_64_a7c9ea4d4c5683aefb0407a6e07adc01) (Utilities|Array|MakeArray "DFX_RightVector")))
  (CallFunction|ApplyTextureData Material))

(fn SetMaterialScalars
  (params (Material :type Material%20Instance%20Dynamic%20Object%20Reference) (Scalars%20Array :type Array%20of%20Float%20%28double-precision%29s) (Params%20Array :type Array%20of%20Names))
  (for (_array_element _array_index) Scalars%20Array
    (Rendering|Material|SetScalarParameterValue Material (Utilities|Array|Get(aref) Params%20Array _array_index) _array_element :!class /Script/Engine.MaterialInstanceDynamic)))

(fn SetMaterialVectors
  (params (Material :type Material%20Instance%20Dynamic%20Object%20Reference) (Vectors%20Array :type Array%20of%20Linear%20Color%20Structures) (Params%20Array :type Array%20of%20Names))
  (for (_array_element _array_index) Vectors%20Array
    (Rendering|Material|SetVectorParameterValue Material (Utilities|Array|Get(aref) Params%20Array _array_index) _array_element :!class /Script/Engine.MaterialInstanceDynamic)))

(fn SetScalarMaterialPrimitives
  (params (PrimitiveIdArray :type Array%20of%20Integers) (ScalarArray :type Array%20of%20Float%20%28double-precision%29s))
  (for (_array_element _array_index) PrimitiveIdArray
    (Rendering|Material|SetDefaultCustomPrimitiveDataFloat (Variables|Default|GetStaticMeshComponent) _array_element (Utilities|Array|Get(acopy) ScalarArray _array_index))))

(fn SetVectorMaterialPrimitives
  (params (PrimitiveIdArray :type Array%20of%20Integers) (VectorsArray :type Array%20of%20Linear%20Color%20Structures))
  (for (_array_element _array_index) PrimitiveIdArray
    (bind (_output_r _output_g _output_b _output_a) (Math|Color|BreakColor (Utilities|Array|Get(acopy) VectorsArray _array_index)))
    (Rendering|Material|SetDefaultCustomPrimitiveDataVector4 (Variables|Default|GetStaticMeshComponent) _array_element (Math|Vector4|MakeVector4 _output_r _output_g _output_b _output_a))))

(fn UpdateBuildingCustomMaterialParameter
  (bind _returnvalue (Components|GetOwner))
  (bind _asbuilding_static_mesh (Utilities|Casting|CastToBuildingSMActor _returnvalue)
    (:then
      (if (!= (Variables|Default|GetMaterialSlot) 0)
        (Variables|/Script/FortniteGame.BuildingSMActor|SetbAllowCustomMaterial false _asbuilding_static_mesh)
        (Variables|/Script/FortniteGame.BuildingSMActor|SetbConsiderOverrideMaterialsOnAnimationCleanup false _asbuilding_static_mesh)
        (else
          (Variables|/Script/FortniteGame.BuildingSMActor|SetbAllowCustomMaterial true _asbuilding_static_mesh)
          (Variables|/Script/FortniteGame.BuildingSMActor|SetbConsiderOverrideMaterialsOnAnimationCleanup true _asbuilding_static_mesh))))
    (:CastFailed
      (bind _asbuilding_static_mesh__static_geo (Utilities|Casting|CastToBuildingStaticGeo _returnvalue))
      (Variables|/Script/FortniteGame.BuildingStaticGeo|SetbAllowCustomMaterial (== (Variables|Default|GetMaterialSlot) 0) _asbuilding_static_mesh__static_geo))))
