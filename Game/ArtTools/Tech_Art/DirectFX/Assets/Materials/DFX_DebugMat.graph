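# DirectFX debug material, Nanite-enabled and Default Lit, that color-codes projection masks.
# MF_ProjectionCoordinates supplies UVs for a scaled per-axis mask and an unscaled radial mask sized by DFX_FalloffRadius, merged with Max.
# Reroute-driven lerps blend flat debug colors, with one branch mixing screen-space scanlines against green.
# That branch is multiplied by inverse eye adaptation to offset exposure and by inverted Fresnel to fade edges.
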Material /Game/ArtTools/Tech_Art/DirectFX/Assets/Materials/DFX_DebugMat
graph {
 properties {bUsedWithNanite=true, pixelDepthOffsetMode=PDOM_Legacy, usedShadingModels=Default Lit}
 nodes [
  LinearInterpolate()
  LinearInterpolate(
   A=LinearInterpolate(A=Constant3Vector() {constant=<rgba 0.2387 0.599 0 1>}, B=Constant3Vector() {constant=<rgba 0.5709 0 0.599 1>}, Alpha=NamedRerouteUsage())
   B=Constant3Vector() {constant=<rgba 0.2157 0.5177 0.8281 1>}
   Alpha=NamedRerouteUsage()
  )
  n1 = Divide(A=ScreenPosition().PixelPosition) {constB=250}
  n2 = Multiply(
   A=Constant3Vector() {constant=<rgba 0.599 0 0.2194 1>}
   B=Clamp(None=MaterialFunctionCall!(ScanLine_Tiling_1, Scanline_PannerSpeed_1, ScanLine_Tiling, Scanline_PannerSpeed, UVs=@n1) {materialFunction=/Game/Weapons/WeaponSkins/MaterialFunction/ScanLines}.Result) {minDefault=0.65}
  )
  n3 = Constant3Vector() {constant=<rgba 0.2387 0.599 0 1>}
  n4 = NamedRerouteUsage()
  Multiply(
   A=Multiply(A=LinearInterpolate(A=@n2, B=@n3, Alpha=@n4), B=EyeAdaptationInverse().EyeAdaptationInverse)
   B=OneMinus(None=Fresnel() {baseReflectFraction=0})
  )
  n5 = MaterialFunctionCall!(World Position) {materialFunction=/Game/ArtTools/Tech_Art/DirectFX/Assets/MaterialFunctions/MF_ProjectionCoordinates}
  n6 = MaterialFunctionCall(Float2=@n5.UVCoord) {materialFunction=&BreakOutFloat2Components}
  n7 = Constant() {r=0.5}
  n8 = Multiply(A=SphereMask(A=@n6.R, B=@n7) {attenuationRadius=0.5}, B=SphereMask(A=@n6.G, B=@n7) {attenuationRadius=0.5})
  n9 = SphereMask(
   A=@n5.UVCoord Unscaled
   B=Constant2Vector() {g=0.5, r=0.5}
   Radius=ScalarParameter() {defaultValue=1, group=DFX, parameterName=DFX_FalloffRadius}
  )
  NamedRerouteDeclaration(None=Max(
    A=NamedRerouteDeclaration(None=@n8) {name=Scaled Mask, nodeColor=<rgba 1 0.6563 0 1>}
    B=NamedRerouteDeclaration(None=@n9) {name=Unscaled Mask, nodeColor=<rgba 1 0 0.5625 1>}
   )) {name=Masks, nodeColor=<rgba 0.4297 1 0 1>}
 ]
}
