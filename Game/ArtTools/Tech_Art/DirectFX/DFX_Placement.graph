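# DFX_Placement is an editor-utility placement Blueprint for the DirectFX art tool, holding settings for projected effects such as decals and, judging by its textures, caustics.
# Variables cover projection type, actor selection, textures, tiling, blending and material response.
# GetTextures synchronously loads the soft-referenced decal, caustics and NMRE textures and returns them as Texture2D.
# The construction script resets the Arrow's world scale to one; the event graph is empty, so no projection logic lives here.
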EditorUtilityBlueprint /Game/ArtTools/Tech_Art/DirectFX/DFX_Placement
graphs [
 EventGraph {links [], nodes []}
 GetTextures {nodes [
   GetTextures() {is=title K2Node_FunctionEntry}
   > n1 = Sequence!(execute) -> (then_0, then_1, then_2:Exec, then_3:Exec)
   > n2 = LoadAsset_Blocking(Asset=Get DecalTexture() -> (DecalTexture:Texture 2D Soft Object Reference))
   > n3 = Cast To Texture2D(Object=@n2)
   > Set LDecalTexture(LDecalTexture:Texture 2D Object Reference=@n3) -> (Output_Get:Texture 2D Object Reference)
   n4 = LoadAsset_Blocking(execute=@n1.then_1, Asset=Get CausticsTexture() -> (CausticsTexture:Texture 2D Soft Object Reference))
   > n5 = Cast To Texture2D(Object=@n4)
   > Set LCausticsTexture(LCausticsTexture:Texture 2D Object Reference=@n5) -> (Output_Get:Texture 2D Object Reference)
   n6 = LoadAsset_Blocking(execute=@n1.then_2, Asset=Get NMRE Texture() -> (NMRE Texture:Texture 2D Soft Object Reference))
   > n7 = Cast To Texture2D(Object=@n6)
   > Set LNSRTexture(LNSRTexture:Texture 2D Object Reference=@n7) -> (Output_Get:Texture 2D Object Reference)
   Return Node!(
    execute=@n1.then_3
    DecalTexture:Texture 2D Object Reference=Get LDecalTexture() -> (LDecalTexture:Texture 2D Object Reference)
    CausticTexture:Texture 2D Object Reference=Get LCausticsTexture() -> (LCausticsTexture:Texture 2D Object Reference)
    NMRE Texture:Texture 2D Object Reference=Get LNSRTexture() -> (LNSRTexture:Texture 2D Object Reference)
   )
  ]}
 UserConstructionScript {nodes [
   Construction Script()
   > SetWorldScale3D!(execute, self:Scene Component Object Reference=Get Arrow() -> (Arrow:Arrow Component Object Reference), NewScale:Vector="1,1,1") -> (then)
  ]}
]
variables [
 Actors {type="array<Actor>"}
 AlphaContrast {type=number, default=1}
 AlphaSource {type=E_DFX_AlphaSource, default=A}
 AngleBlend {type=number, default=0.5}
 AngleBlendHardness {type=number, default=0.5}
 CausticsTexture {type=Texture2D, default=/FortWater/Textures/T_Caustic_M}
 DecalTexture {type=Texture2D, default=/Game/Environments/Asteria/Sets/IndustrialTown/Textures/Decals/T_IndustrialTown_Graffiti_4x4_C}
 Emissive {type=number, default=4}
 Falloff {type=number, default=1}
 FalloffRadius {type=number, default=1}
 Intensity {type=number, default=1}
 InvertAlpha {type=boolean}
 MaterialSlot {type=integer}
 Metallic {type=number, default=1}
 NMRE Texture {type=Texture2D}
 Normal {type=number, default=0.5}
 Opacity {type=number, default=1}
 OverlappingActors {type="array<Actor>"}
 PanningSpeed {type=LinearColor, default=<rgba -0.05 0.02 0.02 0.05>}
 Perspective {type=number}
 ProjectionFunction {type=Enum_DirectFX_Type, default=Decal}
 Roughness {type=number, default=1}
 SelectionMode {type=Enum_DirectFX_SelectionMode, default=Manual}
 SortPriority {type=integer}
 Specular {type=number, default=1}
 SpecularOverride {type=number, default=0.5}
 Tiling {type=number, default=1}
 TilingX {type=number, default=1}
 TilingY {type=number, default=1}
 TintColor {type=LinearColor, default=<rgba 1 1 1 1>}
 UseEmissive {type=boolean}
]
