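# DFX_CausticsMenu is an Editor Utility Widget that serves as the caustics settings sub-menu within the DirectFX tech art tools.
# A StackBox holds a single DetailsView filtered to caustics properties, covering texture, projection angle blending, perspective, tiling, intensity, falloff, tint and panning.
# The widget has no variables or implemented graph logic, and nothing here sets the object the DetailsView edits.
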EditorUtilityWidgetBlueprint /Game/ArtTools/Tech_Art/DirectFX/SubWidgets/DFX_CausticsMenu
graphs [EventGraph {nodes [
   Event PreConstruct()
   Event Construct()
   Event Tick()
  ]}]
variables []
widgets {
 info {rootWidgetClass=&StackBox, widgetCount=2}
 widgets [
  StackBox_0 {widgetClassPath=&StackBox, widget=StackBox_0}
  DetailsView_41 {
   widgetClassPath=&DetailsView
   parent=StackBox_0
   widget=DetailsView_41
   properties {propertiesToShow [CausticsTexture, AngleBlend, AngleBlendHardness, Perspective, Tiling, Intensity, Falloff, FalloffRadius, TintColor, PanningSpeed]}
   slot {class=StackBoxSlot, properties {verticalAlignment=VAlign_Top, size {sizeRule=Fill, value=1}}}
  }
 ]
}
