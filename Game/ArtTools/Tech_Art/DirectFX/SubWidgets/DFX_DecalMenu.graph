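# DFX_DecalMenu is an editor utility sub-widget of the DirectFX tool that displays decal properties for editing.
# Stacked DetailsViews expose texture, alpha, blend, tiling, tint, metallic and roughness/normal settings, with the emissive view and separate alpha property rows collapsed.
# Changes in the decal texture view fire OnPropertyChange, while a change to NMRE Texture fires OnNSRTextureChanged.
# The graph only broadcasts these dispatchers and does not assign the objects being viewed.
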EditorUtilityWidgetBlueprint /Game/ArtTools/Tech_Art/DirectFX/SubWidgets/DFX_DecalMenu
graphs [
 EventGraph {nodes [
   n1 = "On Property Changed (DetailsView_41)"!() -> (OutputDelegate, then, PropertyName:Name=None) {is=title K2Node_ComponentBoundEvent}
   > Branch(Condition="Equal (Name)"(A=@n1.PropertyName, B=NMRE Texture))
   > Call On NSRTexture Changed!(execute:Exec, self:DFX Decal Menu Object Reference) -> (then:Exec) {is=title K2Node_CallDelegate}
   "On Property Changed (DecalTextureView)"!() -> (OutputDelegate, then, PropertyName:Name=None) {is=title K2Node_ComponentBoundEvent}
   > Call On Property Change!(execute:Exec, self:DFX Decal Menu Object Reference) -> (then:Exec) {is=title K2Node_CallDelegate}
   Get DetailsViewNSR() -> (DetailsViewNSR:Details View Object Reference)
   Get MetallicView() -> (MetallicView:Details View Object Reference)
  ]}
 OnNSRTextureChanged {nodes [OnNSRTextureChanged() {is=title K2Node_FunctionEntry}]}
 OnPropertyChange {nodes [OnPropertyChange() {is=title K2Node_FunctionEntry}]}
]
variables [
 OnNSRTextureChanged {default=()}
 OnPropertyChange {default=()}
]
widgets {
 info {rootWidgetClass=&StackBox, widgetCount=23}
 widgets [
  StackBox_0 {widgetClassPath=&StackBox, widget=StackBox_0}
  VerticalBox_0 {
   widgetClassPath=&VerticalBox
   parent=StackBox_0
   widget=VerticalBox_0
   slot {class=StackBoxSlot, properties {verticalAlignment=VAlign_Top, size {sizeRule=Fill, value=1}}}
  }
  Overlay_4 {widgetClassPath=&Overlay, parent=VerticalBox_0, widget=Overlay_4, slot {class=VerticalBoxSlot}}
  DecalTextureView {
   widgetClassPath=&DetailsView
   parent=Overlay_4
   widget=DecalTextureView
   properties {bShowScrollBar=false, propertiesToShow [DecalTexture, AlphaSource, AlphaContrast, InvertAlpha]}
   slot {class=OverlaySlot, properties {horizontalAlignment=HAlign_Fill, verticalAlignment=VAlign_Center}}
  }
  HorizontalBox_1 {
   widgetClassPath=&HorizontalBox
   parent=VerticalBox_0
   widget=HorizontalBox_1
   properties {clipping=ClipToBounds, pixelSnapping=Disabled, visibility=Collapsed}
   slot {class=VerticalBoxSlot, properties {verticalAlignment=VAlign_Top}}
  }
  VerticalBox_1 {
   widgetClassPath=&VerticalBox
   parent=HorizontalBox_1
   widget=VerticalBox_1
   slot {class=HorizontalBoxSlot, properties {verticalAlignment=VAlign_Center, size {sizeRule=Fill, value=0.2}}}
  }
  HorizontalBox_0 {widgetClassPath=&HorizontalBox, parent=VerticalBox_1, widget=HorizontalBox_0, slot {class=VerticalBoxSlot}}
  Border_0 {
   widgetClassPath=&Border
   parent=HorizontalBox_0
   widget=Border_0
   properties {brushColor=<rgba 1 1 1 0>, padding=<ltrb 4 2 4 0>}
   slot {class=HorizontalBoxSlot, properties {verticalAlignment=VAlign_Center, size {sizeRule=Fill, value=1}}}
  }
  AlphaSourceView {
   widgetClassPath=&SinglePropertyView
   parent=Border_0
   widget=AlphaSourceView
   properties {propertyName=AlphaSource}
   slot {class=BorderSlot, properties {padding=<ltrb 4 2 4 0>}}
  }
  HorizontalBox_2 {widgetClassPath=&HorizontalBox, parent=VerticalBox_1, widget=HorizontalBox_2, slot {class=VerticalBoxSlot}}
  Border_4 {
   widgetClassPath=&Border
   parent=HorizontalBox_2
   widget=Border_4
   properties {brushColor=<rgba 1 1 1 0>, padding=<ltrb 4 2 4 0>}
   slot {class=HorizontalBoxSlot, properties {verticalAlignment=VAlign_Center, size {sizeRule=Fill, value=1}}}
  }
  AlphaContrastView {
   widgetClassPath=&SinglePropertyView
   parent=Border_4
   widget=AlphaContrastView
   properties {propertyName=AlphaContrast}
   slot {class=BorderSlot, properties {padding=<ltrb 4 2 4 0>}}
  }
  HorizontalBox {widgetClassPath=&HorizontalBox, parent=VerticalBox_1, widget=HorizontalBox, slot {class=VerticalBoxSlot}}
  Border_2 {
   widgetClassPath=&Border
   parent=HorizontalBox
   widget=Border_2
   properties {brushColor=<rgba 1 1 1 0>, padding=<ltrb 4 2 4 0>}
   slot {class=HorizontalBoxSlot, properties {verticalAlignment=VAlign_Center, size {sizeRule=Fill, value=1}}}
  }
  InvertAlphaView {
   widgetClassPath=&SinglePropertyView
   parent=Border_2
   widget=InvertAlphaView
   properties {propertyName=InvertAlpha}
   slot {class=BorderSlot, properties {padding=<ltrb 4 2 4 0>}}
  }
  Overlay_5 {
   widgetClassPath=&Overlay
   parent=VerticalBox_0
   widget=Overlay_5
   properties {clipping=ClipToBounds}
   slot {class=VerticalBoxSlot}
  }
  DetailsView_41 {
   widgetClassPath=&DetailsView
   parent=Overlay_5
   widget=DetailsView_41
   properties {
    clipping=ClipToBounds
    propertiesToShow [NMRE Texture, AngleBlend, AngleBlendHardness, Perspective, Tiling, TilingX, TilingY, Opacity, TintColor]
   }
   slot {class=OverlaySlot, properties {horizontalAlignment=HAlign_Fill, verticalAlignment=VAlign_Fill}}
  }
  Overlay_2 {
   widgetClassPath=&Overlay
   parent=VerticalBox_0
   widget=Overlay_2
   properties {clipping=ClipToBounds}
   slot {class=VerticalBoxSlot}
  }
  MetallicView {
   widgetClassPath=&DetailsView
   parent=Overlay_2
   widget=MetallicView
   properties {bOverride_Cursor=true, propertiesToShow [Metallic]}
   slot {class=OverlaySlot, properties {horizontalAlignment=HAlign_Fill, verticalAlignment=VAlign_Fill, padding=<ltrb 0 -28.5 0 0>}}
  }
  Overlay_1 {
   widgetClassPath=&Overlay
   parent=VerticalBox_0
   widget=Overlay_1
   properties {clipping=ClipToBounds}
   slot {class=VerticalBoxSlot}
  }
  EmissiveView {
   widgetClassPath=&DetailsView
   parent=Overlay_1
   widget=EmissiveView
   properties {bOverride_Cursor=true, visibility=Collapsed, propertiesToShow [Emissive]}
   slot {class=OverlaySlot, properties {horizontalAlignment=HAlign_Fill, verticalAlignment=VAlign_Fill, padding=<ltrb 0 -28.5 0 0>}}
  }
  Overlay_0 {
   widgetClassPath=&Overlay
   parent=VerticalBox_0
   widget=Overlay_0
   properties {clipping=ClipToBounds}
   slot {class=VerticalBoxSlot, properties {verticalAlignment=VAlign_Top, size {sizeRule=Fill, value=1}}}
  }
  DetailsViewNSR {
   widgetClassPath=&DetailsView
   parent=Overlay_0
   widget=DetailsViewNSR
   properties {clipping=ClipToBounds, propertiesToShow [Roughness, SpecularOverride, Normal]}
   slot {class=OverlaySlot, properties {horizontalAlignment=HAlign_Fill, verticalAlignment=VAlign_Fill, padding=<ltrb 0 -28.5 0 0>}}
  }
 ]
}
