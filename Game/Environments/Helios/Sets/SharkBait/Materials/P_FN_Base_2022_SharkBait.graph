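# Base world material for the SharkBait environment set, extending the Fortnite 2022 GigaMix material with an optional world-aligned camo paint overlay.
# A world-space mask tints hull, paint-line and camo areas, adds sand color and roughness breakup, limited by layered exterior paint opacity.
# Shared functions then add destruction, wind, fan spin, snow, biome and RVT blending.
# Normals convert to world space, and low-quality mobile flattens specular and roughness.
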Material /Game/Environments/Helios/Sets/SharkBait/Materials/P_FN_Base_2022_SharkBait
graph {
 properties {
  bAutomaticallySetUsageInEditor=false
  bTangentSpaceNormal=false
  bUseMaterialAttributes=true
  bUsedWithInstancedStaticMeshes=true
  bUsedWithNanite=true
  bUsedWithSplineMeshes=true
  ditheredLODTransition=true
  pixelDepthOffsetMode=PDOM_Legacy
  shadingModel=MSM_FromMaterialExpression
  usedShadingModels=Default Lit | Subsurface | Clear Coat | Two Sided Foliage
 }
 nodes [
  n1 = MaterialFunctionCall() {materialFunction=/Game/Global/Base4Layers_2022/MaterialFunctions/MF_GigaMix_2023}
  n2 = VectorParameter() {group=SharkBait POI, parameterName=Hull Color, defaultValue=<rgba 0.6823 0.6823 0.6823 1>}
  n3 = MaterialFunctionCall(
   TextureObject=TextureObject() {samplerType=SAMPLERTYPE_Masks, texture=/Game/Environments/Helios/Sets/SharkBait/Textures/T_SharkBait_Camoflage_Mask}
   TextureSize=ScalarParameter() {defaultValue=750, group=SharkBait POI, parameterName=SharkBait Camo Tiling}
   WorldPosition=Add(A=VectorParameter() {group=SharkBait POI, parameterName=Adjust World Offset, defaultValue=<rgba 0 1500 0 0>}.RGB, B=WorldPosition().XYZ)
   Export Float 4=StaticBool() {value=true}
  ) {materialFunction=&WorldAlignedTexture}
  n4 = MaterialFunctionCall(Color=@n3.XY Texture) {materialFunction=&SplitComponents}
  n5 = Saturate(None=MaterialFunctionCall!(
    1=ScalarParameter() {defaultValue=1, group=SharkBait POI, parameterName=SharkBait Exterior Paint Layer 1 Opacity}
    2=ScalarParameter() {group=SharkBait POI, parameterName=SharkBait Exterior Paint Layer 2 Opacity}
    3=ScalarParameter() {group=SharkBait POI, parameterName=SharkBait Exterior Paint Layer 3 Opacity}
    4=ScalarParameter() {group=SharkBait POI, parameterName=SharkBait Exterior Paint Layer 4 Opacity}
   ) {materialFunction=/Game/Global/Base4Layers_2022/MaterialFunctions/MF_UV2LerpScalar_Layers_New}.Result)
  n6 = MaterialFunctionCall!(
   Material=@n1.Result
   Tint=LinearInterpolate(
    A=LinearInterpolate(B=@n2.RGB, Alpha=@n4.G)
    B=VectorParameter() {group=SharkBait POI, parameterName=Paint Line Color, defaultValue=<rgba 1 0 0 1>}.RGB
    Alpha=@n4.R
   )
   TintMask=@n5
  ) {materialFunction=/Engine/Functions/MaterialLayerFunctions/MatLayerBlend_Tint}
  n7 = GetMaterialAttributes(None=@n6.Blended Material) {attributeGetTypes [#1, #2]}
  n8 = LinearInterpolate(
   A=@n7.Base Color
   B=VectorParameter() {group=SharkBait POI, parameterName=Sand Color, defaultValue=<rgba 0.4219 0.1453 0.0648 1>}.RGB
   Alpha=@n4.B
  )
  n9 = LinearInterpolate(A=@n7.Roughness, Alpha=@n4.B)
  n10 = LinearInterpolate(
   A=@n9
   B=Add(A=ScalarParameter() {defaultValue=0.5, group=SharkBait POI, parameterName=Large Roughness Breakup Max}, B=@n9)
   Alpha=Multiply(A=@n4.G, B=ComponentMask(None=@n3.XY Texture) {a=true})
  )
  n11 = ScalarParameter() {defaultValue=5, group=SharkBait POI, parameterName=Camo Roughness Contrast}
  n12 = LinearInterpolate(
   A=@n10
   B=Multiply(
    A=MaterialFunctionCall(In=@n4.G, Contrast=@n11) {materialFunction=&CheapContrast}.Result
    B=@n10
   )
   Alpha=ScalarParameter() {defaultValue=0.1, group=SharkBait POI, parameterName=Camo Roughness Contribution}
  )
  n13 = StaticSwitchParameter(
   True=MaterialFunctionCall!(
    Base Material=SetMaterialAttributes!(MaterialAttributes=@n7.MaterialAttributes, Base Color=@n8, Roughness=@n12) {attributeSetTypes [#1, #2]}
    Top Material=@n6.Blended Material
    Alpha=OneMinus(None=@n5)
   ) {materialFunction=/Engine/Functions/MaterialLayerFunctions/MatLayerBlend_Simple}.Blended Material
   False=@n1.Result
  ) {group=SharkBait POI, parameterName=Use WS Camo overlay?}
  MaterialExpressionReroute_70 = Reroute(None=@n13)
  n14 = ScalarParameter() {
   desc=Documentation link
   group="0 (Expand)  Documentation (RMB - Copy display name) and Debug view"
   parameterName=https://confluence.it.epicgames.com/display/FORT/Fortnite+Base+Material+2022
  }
  n15 = SetMaterialAttributes!(MaterialAttributes=@MaterialExpressionReroute_70.MaterialAttributes, World Position Offset=Multiply(A=@n14)) {attributeSetTypes [#3]}
  MaterialExpressionReroute_71 = Reroute(None=@n15)
  MaterialExpressionReroute_69 = Reroute(None=@MaterialExpressionReroute_71.Output)
  n16 = MaterialFunctionCall!(Material Attributes=MaterialFunctionCall!(In=MaterialFunctionCall(Material_In=@MaterialExpressionReroute_69.MaterialAttributes) {materialFunction=&MF_FN_DestructibleDamage}.Material_Out) {materialFunction=/Game/Athena/WinterArt/Materials/MF_WPO_WindNew}.Result) {materialFunction=/Game/Creative/Sets/MilitaryBase/Props/MF_FanSpin}
  n17 = MaterialFunctionCall!(In=MaterialFunctionCall!(
    In=MaterialFunctionCall!(In=@n16.Result, UseSphericalMappingOverride) {materialFunction=/Game/Packages/Fortress_SharedMaterials/Base_Material_Functions/MF_CurieFallbackEffect}.Result
    Custom Mask
    Custom Rotation Axis
   ) {materialFunction=/Game/Athena/WinterArt/Materials/MF_WindWPO_v2}.Result) {materialFunction=/Game/Global/Base4Layers_2022/MaterialFunctions/MF_SnowAndIceArchitecture}
  MaterialExpressionReroute_68 = Reroute(None=@n17.Result)
  Composite!(None=@MaterialExpressionReroute_68.Result) {desc=Transform to World in case we will use RVT after this point, subgraphName=Tangent to World}
  Composite!(MaterialAttributes=@n13) {subgraphName=Documentation Link}
  Composite(MaterialAttributes=@MaterialExpressionReroute_71.Output) {desc=Game Shared WPO and Fire Effect, subgraphName=WPO And Shared Functions}
  MaterialExpressionReroute_1 = Reroute(None=@n23.RVT H)
  MaterialExpressionReroute_3 = Reroute(None=@MaterialExpressionReroute_68.Result)
  n18 = GetMaterialAttributes(None=@MaterialExpressionReroute_3.Input) {attributeGetTypes [#4]}
  n19 = Transform(None=@n18.Normal)
  n20 = SetMaterialAttributes!(
   MaterialAttributes=@n18.MaterialAttributes
   Normal=StaticSwitchParameter(True=MaterialProxyReplace(Realtime=@n19, MaterialProxy=@n18.Normal), False=@n19) {parameterName=Force TangentSpaceNormalsFor HLOD ProxyReplace}
  ) {attributeSetTypes [#4]}
  MaterialExpressionReroute_2 = Reroute(None=@n20)
  n21 = MaterialFunctionCall!(In=@MaterialExpressionReroute_2.Output) {materialFunction=/Game/ArtTools/Tech_Art/DirectFX/Assets/MaterialFunctions/MF_DirectFX}
  MaterialExpressionReroute_8 = Reroute(None=@n21.Result)
  n22 = MaterialFunctionCall!(AO, RVT_Height=@MaterialExpressionReroute_1."RVT_Height (S)", For Buildings=@MaterialExpressionReroute_8."For Moss (MA)", For Moss) {materialFunction=/Game/Environments/Asteria/Rocks/Materials/MaterialFunctions/MF_JungleBiome}
  MaterialExpressionReroute_0 = Reroute(None=@n22.Buildings Jungl)
  n23 = MaterialFunctionCall!(
   In=@MaterialExpressionReroute_0.Result
   "RVT Height Map (Roughness)"=GetMaterialAttributes(None=@n21.Result) {attributeGetTypes [#2]}.Roughness
  ) {materialFunction=/Game/Global/Base4Layers_2022/MaterialFunctions/MF_RVTBlending}
  Composite!(InputPin=@n23.Result) {subgraphName=Mobile Opt}
  Composite!("For Moss (MA)"=@n21.Result, "RVT_Height (S)"=@n23.RVT H) {subgraphName="Seasonal "}
  MaterialExpressionReroute_7 = Reroute(None=@n23.Result)
  n24 = GetMaterialAttributes(None=@MaterialExpressionReroute_7.InputPin)
  n25 = Constant() {r=0.5}
  n26 = Constant() {r=0.8}
  n27 = ShadingPathSwitch(
   Default=@MaterialExpressionReroute_7.InputPin
   Mobile=QualitySwitch(
    Default=@MaterialExpressionReroute_7.InputPin
    Low=SetMaterialAttributes!(MaterialAttributes=@n24.MaterialAttributes, Specular=@n25, Roughness=@n26) {attributeSetTypes [#5, #2]}
    High=@MaterialExpressionReroute_7.InputPin
    Medium=@MaterialExpressionReroute_7.InputPin
    Epic=@MaterialExpressionReroute_7.InputPin
   )
  )
  MaterialExpressionReroute_6 = Reroute(None=@n27)
  MaterialFunctionCall(MA In=@MaterialExpressionReroute_6.Output) {materialFunction=&MF_CreativeMoveTool}
  PinBase() {reroutePins [Input {expression=&self:MaterialExpressionReroute_3}]}
  PinBase!(Output=@n20) {reroutePins [Output {expression=&self:MaterialExpressionReroute_2}]}
  PinBase!() {reroutePins [
    "For Moss (MA)" {expression=&self:MaterialExpressionReroute_8}
    "RVT_Height (S)" {expression=&self:MaterialExpressionReroute_1}
   ]}
  PinBase!(Result=@n22.Buildings Jungl) {reroutePins [Result {expression=&self:MaterialExpressionReroute_0}]}
  PinBase!() {reroutePins [MaterialAttributes {expression=&self:MaterialExpressionReroute_70}]}
  PinBase!(Output=@n15) {reroutePins [Output {expression=&self:MaterialExpressionReroute_71}]}
  PinBase!() {reroutePins [MaterialAttributes {expression=&self:MaterialExpressionReroute_69}]}
  PinBase!(Result=@n17.Result) {reroutePins [Result {expression=&self:MaterialExpressionReroute_68}]}
  PinBase!() {reroutePins [InputPin {expression=&self:MaterialExpressionReroute_7}]}
  PinBase!(Output=@n27) {reroutePins [Output {expression=&self:MaterialExpressionReroute_6}]}
 ]
}
