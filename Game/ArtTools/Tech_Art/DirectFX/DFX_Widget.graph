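# Editor utility widget for the DirectFX art tool, which drives a selected DFX_Placement actor to project decal or caustic effects onto building actors.
# Tick tracks the placement and, on changed parameters or forced updates, gathers overlapping and pinned actors, then refreshes their DFX material components and parent materials.
# Missing materials are generated by duplicating an empty MIC, parenting it to each mesh's material, setting DirectFX static switches and mirroring Nanite overrides.
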EditorUtilityWidgetBlueprint /Game/ArtTools/Tech_Art/DirectFX/DFX_Widget
graphs [
 Add DFX Component {nodes [
   n1 = Add DFX Component!() -> (then, Actor:Actor Object Reference) {is=title K2Node_FunctionEntry}
   n2 = Get SubobjectDataSubsystem()
   > n3 = Gather Subobject Data For Instance(self=@n2, Context=@n1)
   > AddNewSubobject!(
    execute
    self:Subobject Data Subsystem Object Reference=@n2
    Params_ParentHandle:Subobject Data Handle Structure="Get (a ref)"(Array:Array of Subobject Data Handle Structures=@n3, Dimension 1=0) -> (Output:"Subobject Data Handle Structure (by ref)")
    Params_NewClass:Object Class Reference=&/DFX_MaterialComponent.DFX_MaterialComponent_C
    Params_BlueprintContext:Blueprint Object Reference
    Params_bSkipMarkBlueprintModified:bool=False
    Params_bConformTransformToParent:bool=false
   ) -> (then, FailReason:Text, ReturnValue:Subobject Data Handle Structure)
  ]}
 AddUnselectedOverlappingActor {nodes [
   n1 = Inputs!() -> (execute, InActor:"Actor Object Reference (by ref)")
   n2 = Get DFX_Placement() -> (DFX_Placement:DFX Placement Object Reference)
   > n3 = Branch(Condition=Contains Item(
     TargetArray:Array of Actor Object References=Get Actors!(self:DFX Placement Object Reference=@n2) -> (Actors:Array of Actor Object References)
     ItemToFind:"Actor Object Reference (by ref)"=@n1
    ))
   Add(
    execute=@n3.else
    TargetArray:Array of Actor Object References=Get OverlappingActors!(self:DFX Placement Object Reference=Get DFX_Placement()) -> (OverlappingActors:Array of Actor Object References)
    NewItem:"Actor Object Reference (by ref)"=@n1
   ) {is=title K2Node_CallArrayFunction}
   > Outputs!(then:Exec)
  ]}
 AssignParentMaterialToComponent {nodes [
   AssignParentMaterialToComponent() {is=title K2Node_FunctionEntry}
   > n1 = For Each Loop(Array:Array of DFX Material Component Object References=Get DFXComponents() -> (DFXComponents:Array of DFX Material Component Object References)) -> (Array Element:DFX Material Component Object Reference)
   > Is Valid(InputObject=@n1)
   n2 = GetOwner(self=@n1)
   n3 = Find(
    TargetMap:Map of Static Mesh Object References to Material Interface Object References=Get BuildingActorsMeshes() -> (BuildingActorsMeshes:Map of Static Mesh Object References to Material Interface Object References)
    Key:"Static Mesh Object Reference (by ref)"=Get StaticMesh(self=GetComponentByClass!(self:Actor Object Reference=@n2, ComponentClass:Actor Component Class Reference=&StaticMeshComponent) -> (ReturnValue:Static Mesh Component Object Reference))
   ) -> (Value:Material Interface Object Reference)
   n4 = DFX Material Path!(Mat Interface:Material Interface Object Reference=@n3) -> (Full Path:String, Name:String)
   > n5 = DoesAssetExist!(
    execute
    self:Editor Asset Subsystem Object Reference=Get EditorAssetSubsystem() -> (ReturnValue:Editor Asset Subsystem Object Reference) {is=title K2Node_GetEditorSubsystem}
    AssetPath=@n4
   ) -> (then, ReturnValue)
   > Branch(Condition=@n5)
   > n6 = LoadAsset(AssetPath=@n4)
   > n7 = Cast To MaterialInterface!(execute, Object=@n6) -> (then, CastFailed, AsMaterial Interface:Material Interface Object Reference) {is=title K2Node_DynamicCast}
   > Set ParentMaterial!(
    execute
    ParentMaterial:Material Interface Object Reference=Select(
     Option 0:Material Interface Object Reference=@n7
     Option 1:Material Interface Object Reference=/Game/ArtTools/Tech_Art/DirectFX/Assets/Materials/DFX_DebugMat
     Index=Get DebugView() -> (DebugView:bool)
    ) -> (ReturnValue:Material Interface Object Reference)
    self:DFX Material Component Object Reference=@n1
   ) -> (then, Output_Get:Material Interface Object Reference)
   "FX Material Path (Deprecated)"!(Name:String=GetDisplayName(Object=@n3)) -> (Full Path:String, ReturnValue:String) {is=title K2Node_MacroInstance}
  ]}
 CheckIfPropertyIsValid {nodes [
   n1 = CheckIfPropertyIsValid!() -> (then, InputPin:Material Interface Object Reference) {is=title K2Node_FunctionEntry}
   > n2 = Cast To MaterialInstanceConstant(Object=@n1)
   > n3 = GetEditorProperty!(execute:Exec, Object:Object Reference=@n2, PropertyName:Name=Enable DirectFX) -> (then:Exec, PropertyValue:bool, ReturnValue:bool)
   > Set NA(NA:bool=@n3) -> (Output_Get:bool)
   > Return Node!(execute, Has Properties:bool=@n3.ReturnValue)
   n4 = Get Object Path String()
   n5 = LeftChop(SourceString=@n4, Count=int - int(A=Len(S=@n4), B=FindSubstring(SearchIn=@n4, Substring=/Materials/, bUseCase=true)))
   RightChop!(SourceString:String=@n5, Count:int=FindSubstring(SearchIn=@n5, Substring=/, bUseCase=true, bSearchFromEnd=true)) -> (ReturnValue:String)
  ]}
 Clear Arrays Data {nodes [
   Clear Arrays Data() {is=title K2Node_FunctionEntry}
   > Clear(TargetArray:Array of Material Interface Object References=Get UniqueMaterials() -> (UniqueMaterials:Array of Material Interface Object References))
   > Clear(TargetArray:Array of Actor Object References=Get BuildingActors() -> (BuildingActors:Array of Actor Object References))
   > Clear(TargetArray:Array of DFX Material Component Object References=Get DFXComponents())
   > Clear(TargetArray:Array of Material Interface Object References=Get ParentMaterialToUseForActor() -> (ParentMaterialToUseForActor:Array of Material Interface Object References))
   > Clear!(
    execute
    TargetMap:Map of Static Mesh Object References to DFX Mesh Materials Structures=Get UniqueStaticMeshes() -> (UniqueStaticMeshes:Map of Static Mesh Object References to DFX Mesh Materials Structures)
   ) -> (then) {is=title K2Node_CallFunction}
   > Clear(TargetMap:Map of Static Mesh Object References to Material Interface Object References=Get BuildingActorsMeshes())
   > Clear(TargetArray=Get UniqueMaterialConstants() -> (UniqueMaterialConstants:Array of Material Interface Object References)) {is=title K2Node_CallArrayFunction}
   > Clear(TargetArray=Get Queued Materials To Create() -> (Queued Materials To Create:Array of Material Interface Object References))
  ]}
 Create Missing Materials {nodes [
   Create Missing Materials() {is=title K2Node_FunctionEntry}
   > n1 = For Each Loop(Array:Array of Material Interface Object References=Get Queued Materials To Create()) -> (Array Element:Material Interface Object Reference)
   > n2 = GetBaseMaterial!(execute, self:Material Interface Object Reference=@n1) -> (then, ReturnValue:Material Object Reference)
   > n3 = GetStaticSwitchParameterSource!(execute, Material:Material Interface Object Reference=@n2, ParameterName:Name=Enable DirectFX) -> (then, ParameterSource:Soft Object Path Structure, ReturnValue:bool)
   > Branch(Condition=@n3.ReturnValue)
   > n4 = Sequence()
   > n5 = Create Missing Materials Macro!(execute:Exec, Material:Material Interface Object Reference=@n1) -> (output:Exec, OutputMaterial:Material Instance Constant Object Reference) {is=title K2Node_MacroInstance}
   > Set LDFXMaterial(LDFXMaterial:Material Instance Constant Object Reference=@n5) -> (Output_Get:Material Instance Constant Object Reference)
   n6 = GetNaniteOverideMaterial!(self:Material Interface Object Reference=@n1) -> (ReturnValue:Material Interface Object Reference)
   n7 = Is Valid(exec=@n4.then_1, InputObject=@n6)
   > n8 = Create Missing Materials Macro(Material=@n6)
   > n9 = SetNaniteOverrideMaterial!(
    execute
    self:Material Instance Constant Object Reference=Get LDFXMaterial() -> (LDFXMaterial:Material Instance Constant Object Reference)
    bInEnableOverride:bool=true
    InOverrideMaterial:Material Interface Object Reference=@n8
   ) -> (then)
   SaveLoadedAsset(execute=[@n9.then, @n7.Is Not Valid], AssetToSave=Get LDFXMaterial(), bOnlyIfIsDirty=false)
  ]}
 CreateMissingMaterialsMacro {nodes [
   n1 = Inputs!() -> (execute, Material:Material Interface Object Reference)
   n2 = DFX Material Path(Mat Interface=@n1)
   n3 = Get EditorAssetSubsystem()
   > n4 = DoesAssetExist(self=@n3, AssetPath=@n2)
   > n5 = Branch(Condition=@n4)
   > n6 = LoadAsset!(execute, self:Editor Asset Subsystem Object Reference=@n3, AssetPath=@n2) -> (then, ReturnValue)
   > n7 = Set DFXMaterialTemp(DFXMaterialTemp:Object Reference=@n6) -> (Output_Get:Object Reference)
   n8 = DuplicateAsset!(
    execute=@n5.else
    self:Editor Asset Subsystem Object Reference=Get EditorAssetSubsystem()
    SourceAssetPath:String=/Game/ArtTools/Tech_Art/DirectFX/Assets/Materials/DFX_EmptyMIC
    DestinationAssetPath:String=DFX Material Path(Mat Interface=@n1)
   ) -> (then, ReturnValue:Object Reference)
   > n9 = Set DFXMaterialTemp(DFXMaterialTemp=@n8)
   n10 = Cast To MaterialInstanceConstant(execute=[@n9.then, @n7.then], Object=@n9)
   > SetMaterialInstanceParent!(execute, Instance:Material Instance Constant Object Reference=@n10, NewParent:Material Interface Object Reference=@n1) -> (then)
   > SetMaterialInstanceStaticSwitchParameterValue!(
    execute
    Instance:Material Instance Constant Object Reference=@n10
    ParameterName:Name=Enable DirectFX
    Value:bool=true
    Association:EMaterialParameterAssociation Enum=GlobalParameter
    bUpdateMaterialInstance:bool=true
   ) -> (then, ReturnValue:bool)
   n11 = Get DFX_Placement()
   > SetMaterialInstanceStaticSwitchParameterValue(
    Instance=@n10
    ParameterName=Select!(
     NewEnumerator0:Name=Enable Decal
     NewEnumerator1:Name=Enable Caustics
     Index:Enum_DirectFX_Type Enum=Get ProjectionFunction!(self:DFX Placement Object Reference=@n11) -> (ProjectionFunction:Enum_DirectFX_Type Enum=NewEnumerator0)
    ) -> (ReturnValue:Name)
    Value=true
    Association=GlobalParameter
    bUpdateMaterialInstance=true
   )
   n12 = Get DFX_Placement()
   > SetMaterialInstanceStaticSwitchParameterValue(
    Instance=@n10
    ParameterName=UseNSRTexture
    Value=IsValid(Object=GetTextures!(self:DFX Placement Object Reference=@n12) -> (DecalTexture:Texture 2D Object Reference, CausticTexture:Texture 2D Object Reference, NMRE Texture:Texture 2D Object Reference).NMRE Texture)
    Association=GlobalParameter
    bUpdateMaterialInstance=true
   )
   > Outputs!(output:Exec, OutputMaterial:Material Instance Constant Object Reference=@n10)
  ]}
 DFX Actor Updated {nodes [
   Inputs!() -> (execute)
   > n1 = Branch(Condition=Get bForceUpdate() -> (bForceUpdate:bool))
   > n2 = Set bForceUpdate(bForceUpdate:bool=false) -> (Output_Get:bool)
   Outputs!(true:Exec=[@n6.then, @n2.then, @n7.then])
   n3 = Set MaterialParamsStruct(
    execute=@n1.else
    MaterialParamsStruct:DFX Material Params Struct Structure=Make Material Struct!() -> (MaterialParamsStruct:DFX Material Params Struct Structure) {is=title K2Node_MacroInstance}
   ) -> (Output_Get:DFX Material Params Struct Structure)
   n4 = Get MaterialParamsStruct_Stored() -> (MaterialParamsStruct_Stored:DFX Material Params Struct Structure)
   > n5 = Branch(Condition=Contains Item(
     TargetArray:Array of DFX Material Params Struct Structures=Make Array("[0]":DFX Material Params Struct Structure=@n4) -> (Array:Array of DFX Material Params Struct Structures)
     ItemToFind:"DFX Material Params Struct Structure (by ref)"=@n3
    ))
   > n6 = Branch(Condition=Contains Item(
     TargetArray:Array of DFX Material Component Object References=Get DFXComponents()
     ItemToFind:"DFX Material Component Object Reference (by ref)"
    ))
   n7 = Set MaterialParamsStruct_Stored(execute=@n5.else, MaterialParamsStruct_Stored:DFX Material Params Struct Structure=@n3) -> (Output_Get:DFX Material Params Struct Structure)
  ]}
 DFX Material Path {nodes [
   n1 = DFX Material Path!() -> (then, Mat Interface:Material Interface Object Reference) {is=title K2Node_FunctionEntry}
   > Set lMatInterface(lMatInterface:Material Interface Object Reference=@n1) -> (Output_Get:Material Interface Object Reference)
   > n2 = GetPathNameForLoadedAsset(LoadedAsset=Get lMatInterface() -> (lMatInterface:Material Interface Object Reference))
   > Set lPath(lPath:String=@n2) -> (Output_Get:String)
   > n3 = Branch(Condition=Contains(SearchIn=Left!(SourceString:String=@n2, Count:int=9) -> (ReturnValue:String), Substring=/Game/, bUseCase=true))
   n4 = Get DFX_Placement()
   n5 = GetTextures(self=@n4)
   n6 = Append!(
    A=GetDisplayName(Object=Get lMatInterface())
    B=_DFX_
    C:String=Select(NewEnumerator0:String=D, NewEnumerator1:String=C, Index=Get ProjectionFunction(self=@n4) -> (ProjectionFunction=NewEnumerator0)) -> (ReturnValue:String)
    D:String=SelectString(A=_NRME, bPickA=IsValid(Object=@n5.NMRE Texture))
   ) -> (ReturnValue)
   > Return Node!(execute, Full Path:String=Append!(A=Get MICsPath() -> (MICsPath:String), B=@n6) -> (ReturnValue), Name:String=@n6)
   n7 = Split(SourceString=Get lPath() -> (lPath:String), InStr=/, SearchDir=FromEnd)
   n8 = Get DFX_Placement()
   n9 = GetTextures(self=@n8)
   n10 = Append!(
    A=GetDisplayName(Object=Get lMatInterface())
    B=_DFX_
    C:String=Select(NewEnumerator0=D, NewEnumerator1=C, Index=Get ProjectionFunction(self=@n8) -> (ProjectionFunction=NewEnumerator0))
    D:String=SelectString(A=_NRME, bPickA=IsValid(Object=@n9.NMRE Texture))
   ) -> (ReturnValue)
   Return Node(execute=@n3.else, Full Path=Append!(A=Append(A=@n7, B=/DFX/), B=@n10) -> (ReturnValue), Name=@n10)
  ]}
 DetectMaterialSlotChange {nodes [
   DetectMaterialSlotChange() {is=title K2Node_FunctionEntry}
   > Is Valid(InputObject=Get DFX_Placement())
   n1 = Get MaterialSlotBuffer() -> (MaterialSlotBuffer:int=0)
   n2 = Get MaterialSlot!(self:DFX Placement Object Reference=Get DFX_Placement()) -> (MaterialSlot:int=0)
   > Branch(Condition=OR Boolean(A="Not Equal (Integer)"(A=@n2, B=@n1), B=Get bForceUpdate()))
   > n3 = For Each Loop(Array:Array of DFX Material Component Object References=Get DFXComponents()) -> (Array Element:DFX Material Component Object Reference)
   Set MaterialSlotBuffer(execute=@n3.Completed, MaterialSlotBuffer:int=Get MaterialSlot(self=Get DFX_Placement()) -> (MaterialSlot=0)) -> (Output_Get:int=0)
   ResetMaterials(execute=@n3.LoopBody, self:DFX Material Component Object Reference=@n3)
   > UpdateBuildingCustomMaterialParameter(self:DFX Material Component Object Reference=@n3)
   > n4 = SetDirtyFlag!(
    execute
    self:Editor Asset Subsystem Object Reference=Get EditorAssetSubsystem()
    Object:Object Reference=GetOwner(self=@n3)
    bDirtyState:bool=true
   ) -> (then, ReturnValue:bool)
   > n5 = Branch(Condition=@n4)
   PrintString(execute=@n5.else, InString=ERROR: Failed to set the dirty flag for this DFX Component, TextColor=(R=1,G=0,B=0.0182,A=1))
  ]}
 EventGraph {nodes [
   "On Pressed (CreateMissingMaterialButton)"() {is=title K2Node_ComponentBoundEvent}
   > Find Missing Materials()
   > Create Missing Materials() {is=title K2Node_CallFunction}
   > ForceUpdate()
   "On Pressed (AddBtn)"() {is=title K2Node_ComponentBoundEvent}
   > NewDirectFX()
   "On Clicked (RefreshBtn)"() {is=title K2Node_ComponentBoundEvent}
   > n1 = GetAllActorsOfClass!(
    execute
    WorldContextObject:Object Reference=Self-Reference()
    ActorClass:Actor Class Reference=&/DFX_Placement.DFX_Placement_C
   ) -> (then, OutActors:Array of DFX Placement Object References)
   > n2 = For Each Loop with Break!(Exec:Exec, Array:Array of DFX Placement Object References=@n1, Break:Exec=@n5.then) -> (LoopBody:Exec, Array Element:DFX Placement Object Reference, Array Index:int, Completed:Exec)
   Set bForceUpdate(execute=@n2.Completed, bForceUpdate=true)
   n3 = Branch(execute=@n2.LoopBody, Condition="Equal (Object)"(A=Get DFX_Placement(), B=@n2))
   n4 = Set DFX_Placement(execute=@n3.else, DFX_Placement:DFX Placement Object Reference=@n2) -> (Output_Get:DFX Placement Object Reference)
   > n5 = Set Initilize DFX UI Settings!(execute, self, NewObject:Object Reference=@n4) -> (then) {is=title K2Node_CallFunction}
   "On Property Change (DFX_DecalMenu)"() {is=title K2Node_ComponentBoundEvent}
   > Set bForceUpdate(bForceUpdate=true)
   n6 = "On Check State Changed (EditorUtilityCheckBox_1)"!() -> (OutputDelegate, then, bIsChecked:bool) {is=title K2Node_ComponentBoundEvent}
   n7 = Get DFX_Placement()
   > SetVisibility!(
    execute
    self:Scene Component Object Reference=[
     Get Box!(self:DFX Placement Object Reference=@n7) -> (Box:Box Collision Object Reference)
     Get ProjectionOverlap!(self:DFX Placement Object Reference=@n7) -> (ProjectionOverlap:Static Mesh Component Object Reference)
    ]
    bNewVisibility:bool=@n6.bIsChecked
    bPropagateToChildren:bool=false
   ) -> (then)
   "On Pressed (PinOverlappingActorsBtn)"() {is=title K2Node_ComponentBoundEvent}
   n8 = Get DFX_Placement()
   > Is Valid(InputObject=@n8)
   > Append Array!(
    execute
    TargetArray:Array of Actor Object References=Get Actors(self=@n8)
    SourceArray:Array of Actor Object References=Get OverlappingActors(self=@n8)
   ) -> (then)
   > Set SelectionMode!(execute, SelectionMode:Enum_DirectFX_SelectionMode Enum=NewEnumerator2, self:DFX Placement Object Reference=@n8) -> (then, Output_Get:Enum_DirectFX_SelectionMode Enum=NewEnumerator0)
   "On Clicked (PinSelectedActorsBtn)"() {is=title K2Node_ComponentBoundEvent}
   > Is Valid(InputObject=Get DFX_Placement())
   > n9 = GetSelectedActors()
   > n10 = Set SelectedActors(SelectedActors:Array of Actor Object References=@n9) -> (Output_Get:Array of Actor Object References)
   n11 = For Each Loop(Exec=[@n36.Is Valid, @n10.then], Array:Array of Actor Object References=@n10) -> (Array Element:Actor Object Reference)
   ForceUpdate(execute=@n11.Completed)
   Branch(execute=@n11.LoopBody, Condition=IsBuildingOfAllowedClass!(InActor:Actor Object Reference=@n11) -> (IsAllowedClass:bool))
   > Add Unique(TargetArray:Array of Actor Object References=Get Actors(self=Get DFX_Placement()), NewItem:"Actor Object Reference (by ref)"=@n11)
   "On Clicked (UnpinSelectedActorsBtn)"() {is=title K2Node_ComponentBoundEvent}
   > Is Valid(InputObject=Get DFX_Placement())
   > n12 = GetSelectedActors()
   > n13 = For Each Loop(Array=@n12)
   ForceUpdate(execute=@n13.Completed)
   n14 = Branch(execute=@n13.LoopBody, Condition=IsBuildingOfAllowedClass(InActor=@n13))
   Remove Item!(
    execute=[@n15.CastFailed, @n14.then]
    TargetArray:Array of Actor Object References=Get Actors(self=Get DFX_Placement())
    Item:"Actor Object Reference (by ref)"=@n13
   ) -> (then, ReturnValue:bool)
   n15 = Cast To DFX_Placement!(execute=@n14.else, Object=@n13) -> (then, CastFailed, AsDFX Placement:DFX Placement Object Reference) {is=title K2Node_DynamicCast}
   "On Property Changed (DebugMaterial)"!() -> (OutputDelegate, then, PropertyName:Name=None) {is=title K2Node_ComponentBoundEvent}
   > Set bForceUpdate(bForceUpdate=true)
   "On NSRTexture Changed (DFX_DecalMenu)"() {is=title K2Node_ComponentBoundEvent}
   > n16 = Set bForceUpdate(bForceUpdate=true)
   n17 = Get!(execute:Exec=[@n35.then, @n16.then]) -> (then:Exec, else:Exec, DFX_Placement:DFX Placement Object Reference)
   n18 = IsValid(Object=GetTextures(self=@n17).NMRE Texture)
   n19 = Get DFX_DecalMenu() -> (DFX_DecalMenu:DFX Decal Menu Object Reference)
   > SetIsEnabled(
    self=[
     Get DetailsViewNSR!(self:DFX Decal Menu Object Reference=@n19) -> (DetailsViewNSR:Details View Object Reference)
     Get UseEmissiveCheckbox() -> (UseEmissiveCheckbox:Editor Utility Check Box Object Reference)
     Get EmissiveView!(self:DFX Decal Menu Object Reference=@n19) -> (EmissiveView:Details View Object Reference)
     Get MetallicView!(self:DFX Decal Menu Object Reference=@n19) -> (MetallicView:Details View Object Reference)
    ]
    bInIsEnabled=@n18
   )
   > SetVisibility!(
    execute
    self:Widget Object Reference=Get DirtyMaterialButton() -> (DirtyMaterialButton:Button Object Reference)
    InVisibility:ESlateVisibility Enum=Collapsed
   ) -> (then)
   > "SetText (Text)"(self=Get MaterialDirtyText() -> (MaterialDirtyText:Common Text Object Reference))
   > n20 = For Each Loop(Array:Array of DFX Material Component Object References=Get DFXComponents()) -> (Array Element:DFX Material Component Object Reference)
   Branch(execute=@n20.Completed, Condition=OR Boolean(A=Get bForceUpdate(), B=Get Init() -> (Init:bool)))
   > Set Init(Init:bool=false) -> (Output_Get:bool)
   > Find Missing Materials()
   > Create Missing Materials()
   Set ForceNSRDisabled!(execute=@n20.LoopBody, ForceNSRDisabled:bool=NOT Boolean(A=@n18), self:DFX Material Component Object Reference=@n20) -> (then, Output_Get:bool)
   > n21 = GetDirtyContentPackages!(execute) -> (then, OutDirtyPackages:Array of Package Object References)
   > n22 = For Each Loop(Array:Array of Package Object References=@n21) -> (Array Element:Package Object Reference)
   n23 = Get ParentMaterial!(self:DFX Material Component Object Reference=@n20) -> (ParentMaterial:Material Interface Object Reference)
   > Branch(Condition="Equal (Object)"(A=@n22, B=GetPackageForObject!(Object:Object Reference=@n23) -> (ReturnValue:Package Object Reference)))
   > "SetText (Text)"(
    self=Get MaterialDirtyText()
    InText=Format Text!(
     Format="LOCTEXT(\"{0} {1}\")"
     0:String=MakeLiteralString(Value=-- Don't forget to save this material --)
     1:String=GetDisplayName(Object=@n23)
    ) -> (Result)
   )
   > SetVisibility(self=Get DirtyMaterialButton(), InVisibility=Visible)
   n24 = "On Check State Changed (UseEmissiveCheckbox)"!() -> (OutputDelegate, then, bIsChecked:bool) {is=title K2Node_ComponentBoundEvent}
   > Set UseEmissive(UseEmissive:bool=@n24.bIsChecked) -> (Output_Get:bool)
   > n25 = Sequence()
   > Branch(Condition=Is Not Empty!(TargetArray:Array of DFX Material Component Object References=Get DFXComponents()) -> (ReturnValue:bool))
   > n26 = For Each Loop(Array:Array of DFX Material Component Object References=Get DFXComponents()) -> (Array Element:DFX Material Component Object Reference)
   n27 = Get UseEmissive() -> (UseEmissive:bool)
   > SetMaterialScalars!(
    execute
    self:DFX Material Component Object Reference=@n26
    Material:Material Instance Dynamic Object Reference=Get MID!(self:DFX Material Component Object Reference=@n26) -> (MID:Material Instance Dynamic Object Reference)
    Scalars Array:"Array of Float (double-precision)s"=Make Array("[0]":double="To Float (Boolean)"(InBool=@n27)) -> (Array:"Array of Float (double-precision)s")
    Params Array:Array of Names=Make Array("[0]":Name=DFX_UseEmissive) -> (Array:Array of Names)
   ) -> (then)
   n28 = Sequence!(execute=[@n39.then, @n25.then_1]) -> (then_0, then_1, then_2:Exec)
   > n29 = Get()
   > Set UseEmissive!(execute, UseEmissive=Get UseEmissive(), self:DFX Placement Object Reference=@n29) -> (then, Output_Get)
   n30 = Get DFX_DecalMenu()
   SetVisibility(
    execute=@n28.then_1
    self=Get MetallicView(self=@n30)
    InVisibility=Select!(Option 0:ESlateVisibility Enum=Visible, Option 1:ESlateVisibility Enum=Collapsed, Index:bool=Get UseEmissive()) -> (ReturnValue:ESlateVisibility Enum)
   )
   SetVisibility(execute=@n28.then_2, self=Get EmissiveView(self=@n30), InVisibility=Select(Option 0=Collapsed, Option 1=Visible, Index=Get UseEmissive()))
   "On Clicked (DirtyMaterialButton)"() {is=title K2Node_ComponentBoundEvent}
   > n31 = For Each Loop(Array=Get DFXComponents())
   > n32 = GetDirtyContentPackages()
   > n33 = For Each Loop(Array:Array of Package Object References=@n32) -> (Array Element:Package Object Reference)
   n34 = Get ParentMaterial(self=@n31)
   > Branch(Condition="Equal (Object)"(A=@n33, B=GetPackageForObject(Object=@n34)))
   > SaveLoadedAsset(AssetToSave=@n34, bOnlyIfIsDirty=true)
   > UpdateNSRTextureStatus()
   n35 = UpdateNSRTextureStatus() {is=title K2Node_CustomEvent}
   ForcePinSelectedActors() {is=title K2Node_CustomEvent}
   > n36 = Is Valid(InputObject=Get DFX_Placement())
   UpdateUseEmissive() {is=title K2Node_CustomEvent}
   > n37 = Get()
   > n38 = Set UseEmissive!(execute, UseEmissive=Get UseEmissive!(self:DFX Placement Object Reference=@n37) -> (UseEmissive)) -> (then, Output_Get)
   > n39 = SetIsChecked!(execute, self:Check Box Object Reference=Get UseEmissiveCheckbox(), InIsChecked:bool=@n38) -> (then)
   n40 = Cast To DFX_Placement(Object=@n11)
   Add Unique(execute=@n40.CastFailed, TargetArray=Get Actors(self=Get DFX_Placement()), NewItem=@n11)
   Event PreConstruct()
   > Widget Setup()
   Event Construct()
   > Set Init(Init=true)
   Event Tick()
   > Get DFX Actor() {is=title K2Node_CallFunction}
   > Is Valid(InputObject=Get DFX_Placement())
   > DetectMaterialSlotChange() {is=title K2Node_CallFunction}
   n41 = GetText!() -> (ReturnValue:Text)
   > SetVisibility(
    self=Get MissingMaterials() -> (MissingMaterials:Horizontal Box Object Reference)
    InVisibility=Select(Option 0=Visible, Option 1=Collapsed, Index=TextIsEmpty(InText=@n41))
   )
   > DFX Actor Updated!(execute:Exec) -> (true:Exec) {is=title K2Node_MacroInstance}
   > Clear Arrays Data() {is=title K2Node_CallFunction}
   > Get Affected Actors() {is=title K2Node_CallFunction}
   > Manage DFX Components()
   > UpdateNSRTextureStatus() {is=title K2Node_CallFunction}
   > UpdateUseEmissive() {is=title K2Node_CallFunction}
   > Find Missing Materials()
   > AssignParentMaterialToComponent() {is=title K2Node_CallFunction}
   > Update DFX Components()
  ]}
 "FX Material Path (Deprecated)" {nodes [
   n1 = Get DFX_Placement()
   n2 = Append!(
    A=Inputs!() -> (Name:String)
    B=_DFX_
    C:String=Select!(
     NewEnumerator0:String=D
     NewEnumerator1:String=C
     Index:Enum_DirectFX_Type Enum=Get ProjectionFunction(self=@n1) -> (ProjectionFunction=NewEnumerator0)
    ) -> (ReturnValue:String)
   ) -> (ReturnValue)
   Outputs!(Full Path:String=Append!(A=Get MICsPath(), B=@n2) -> (ReturnValue), ReturnValue:String=@n2)
  ]}
 Find Missing Materials {nodes [
   Find Missing Materials() {is=title K2Node_FunctionEntry}
   > n1 = For Each Loop(Array:Array of Actor Object References=Get BuildingActors()) -> (Array Element:Actor Object Reference)
   n2 = Values!(execute=@n1.Completed, TargetMap:Map of Static Mesh Object References to Material Interface Object References=Get BuildingActorsMeshes()) -> (then, Values:Array of Material Interface Object References)
   > n3 = For Each Loop(Array:Array of Material Interface Object References=@n2) -> (Array Element:Material Interface Object Reference)
   n4 = For Each Loop(Exec=@n3.Completed, Array=Get UniqueMaterialConstants())
   n5 = DFX Material Path(Mat Interface=@n4) {is=title K2Node_CallFunction}
   > n6 = DoesAssetExist(self=Get EditorAssetSubsystem(), AssetPath=@n5)
   > n7 = Branch(Condition=@n6)
   > Is Valid(InputObject=GetNaniteOverideMaterial(self=@n4))
   > n8 = LoadAsset(self=Get EditorAssetSubsystem(), AssetPath=@n5)
   > n9 = Cast To MaterialInstanceConstant(Object=@n8)
   > n10 = Is Valid(InputObject=GetNaniteOverideMaterial(self=@n9))
   Add Unique(
    execute=[@n7.else, @n10.Is Not Valid]
    TargetArray:Array of Material Interface Object References=Get Queued Materials To Create()
    NewItem:"Material Interface Object Reference (by ref)"=@n4
   )
   n11 = Cast To MaterialInstanceConstant(execute=@n3.LoopBody, Object=@n3)
   > Add Unique(TargetArray=Get UniqueMaterialConstants(), NewItem=@n11)
   n12 = Cast To Material!(execute=@n11.CastFailed, Object=@n3) -> (then, CastFailed, AsMaterial:Material Object Reference) {is=title K2Node_DynamicCast}
   > Add Unique(TargetArray=Get UniqueMaterialConstants(), NewItem=@n12)
   n13 = Cast To MaterialInstanceDynamic!(execute=@n12.CastFailed, Object=@n3) -> (then, CastFailed, AsMaterial Instance Dynamic:Material Instance Dynamic Object Reference) {is=title K2Node_DynamicCast}
   > Add Unique(
    TargetArray=Get UniqueMaterialConstants()
    NewItem=Get Parent!(self:Material Instance Object Reference=@n13) -> (Parent:Material Interface Object Reference)
   )
   Is Valid(exec=@n1.LoopBody, InputObject=@n1)
   > n14 = Cast To BuildingSMActor!(execute, Object=@n1) -> (then, CastFailed, AsBuilding Static Mesh:Building Static Mesh Object Reference) {is=title K2Node_DynamicCast}
   > n15 = GetBuildingTextureData!(execute, self:Building Static Mesh Object Reference=@n14) -> (then, OutTextureDataArray:Array of Building Texture Data Object References)
   > Set LTextureData(LTextureData:Array of Building Texture Data Object References=@n15) -> (Output_Get:Array of Building Texture Data Object References)
   > n16 = Set LStaticMeshComponent(LStaticMeshComponent:Static Mesh Component Object Reference=Get StaticMeshComponent(self:Building Static Mesh Object Reference=@n14) -> (StaticMeshComponent:Base Building Static Mesh Component Object Reference)) -> (Output_Get:Static Mesh Component Object Reference)
   n17 = Is Valid(exec=[@n16.then, @n28.then], InputObject=Get LStaticMeshComponent() -> (LStaticMeshComponent:Static Mesh Component Object Reference))
   PrintString(execute=@n17.Is Not Valid, InString=Not able to grab a valid static mesh component., TextColor=(R=1,G=0,B=0.0387,A=1))
   n18 = Get LTextureData() -> (LTextureData:Array of Building Texture Data Object References)
   n19 = Branch(execute=@n17.Is Valid, Condition=Is Not Empty(TargetArray:Array of Building Texture Data Object References=@n18))
   n20 = Get OverrideMaterial!(self:Building Texture Data Object Reference="Get (a copy)"(Array:Array of Building Texture Data Object References=@n18) -> (Output:Building Texture Data Object Reference)) -> (OverrideMaterial:Material Interface Object Reference)
   n21 = Get LStaticMeshComponent()
   n22 = GetNumMaterials!(self:Primitive Component Object Reference=@n21) -> (ReturnValue:int=0)
   n23 = Get DFX_Placement()
   n24 = Get StaticMesh(self=@n21)
   n25 = GetMaterial!(
    self:Static Mesh Object Reference=@n24
    MaterialIndex:int="Min (Integer)"(A=Get MaterialSlot(self=@n23) -> (MaterialSlot=0), B=int - int(A=@n22)) -> (ReturnValue=0)
   ) -> (ReturnValue:Material Interface Object Reference)
   > Add!(
    execute
    TargetMap:Map of Static Mesh Object References to Material Interface Object References=Get BuildingActorsMeshes()
    Key:"Static Mesh Object Reference (by ref)"=@n24
    Value:"Material Interface Object Reference (by ref)"=Select!(Option 0:Material Interface Object Reference=@n25, Option 1:Material Interface Object Reference=@n20, Index:bool=IsValid(Object=@n20)) -> (ReturnValue:Material Interface Object Reference)
   ) -> (then) {is=title K2Node_CallFunction}
   Add(execute=@n19.else, TargetMap=Get BuildingActorsMeshes(), Key=@n24, Value=@n25)
   n26 = Cast To BuildingStaticGeo!(execute=@n14.CastFailed, Object=@n1) -> (then, CastFailed, "AsBuilding Static Mesh (Static Geo)":"Building Static Mesh (Static Geo) Object Reference") {is=title K2Node_DynamicCast}
   > n27 = GetBuildingTextureData(self:"Building Static Mesh (Static Geo) Object Reference"=@n26)
   > Set LTextureData(LTextureData=@n27)
   > n28 = Set LStaticMeshComponent(LStaticMeshComponent=GetComponentByClass(self=@n26, ComponentClass=&StaticMeshComponent))
   n29 = GetComponentByClass(self=@n1, ComponentClass=&StaticMeshComponent)
   Is Valid(exec=@n26.CastFailed, InputObject=@n29)
   > Set LStaticMeshComponent(LStaticMeshComponent=@n29)
   n30 = Get LStaticMeshComponent()
   n31 = int - int(A=GetNumMaterials(self=@n30) -> (ReturnValue=0))
   n32 = Get MaterialSlot(self=Get DFX_Placement()) -> (MaterialSlot=0)
   n33 = Get StaticMesh(self=@n30)
   > Add(
    TargetMap=Get BuildingActorsMeshes()
    Key=@n33
    Value=GetMaterial(self=@n33, MaterialIndex="Min (Integer)"(A=@n32, B=@n31) -> (ReturnValue=0))
   )
  ]}
 ForceUpdate {nodes [
   ForceUpdate() {is=title K2Node_FunctionEntry}
   > Set bForceUpdate(bForceUpdate=true)
  ]}
 Get Affected Actors {nodes [
   Get Affected Actors() {is=title K2Node_FunctionEntry}
   > Clear(TargetArray:Array of Actor Object References=Get OverlappingActors(self=Get DFX_Placement()))
   > Set Submenu Visibility!(execute:Exec) -> (then:Exec) {is=title K2Node_MacroInstance}
   > n1 = Switch on Enum_DirectFX_SelectionMode!(
    execute
    Selection:Enum_DirectFX_SelectionMode Enum=Get SelectionMode!(self:DFX Placement Object Reference=Get DFX_Placement()) -> (SelectionMode:Enum_DirectFX_SelectionMode Enum=NewEnumerator0)
   ) -> (NewEnumerator0:Exec, NewEnumerator1:Exec, NewEnumerator2:Exec) {is=title K2Node_SwitchEnum}
   > n2 = OverlappingActors!(execute, self) -> (then, Actors:Array of Actor Object References)
   > n3 = For Each Loop(Array:Array of Actor Object References=@n2) -> (Array Element:Actor Object Reference)
   > n4 = Cast To BuildingWall!(execute, Object=@n3) -> (then, CastFailed, AsBuilding Wall:Building Wall Object Reference) {is=title K2Node_DynamicCast}
   > Add Unselected Overlapping Actor!(execute:Exec, InActor:"Actor Object Reference (by ref)"=@n4) -> (then:Exec) {is=title K2Node_MacroInstance}
   n5 = Cast To BuildingFloor!(execute=@n4.CastFailed, Object=@n3) -> (then, CastFailed, AsBuilding Floor:Building Floor Object Reference) {is=title K2Node_DynamicCast}
   > Add Unselected Overlapping Actor(InActor=@n5)
   n6 = Cast To Parent_BuildingWall!(execute=@n5.CastFailed, Object=@n3) -> (then, CastFailed, AsParent Building Wall:Parent Building Wall Object Reference) {is=title K2Node_DynamicCast}
   > Add Unselected Overlapping Actor(InActor=@n6)
   Append Array(
    execute=[@n3.Completed, @n8.Completed, @n1.NewEnumerator2]
    TargetArray=Get BuildingActors()
    SourceArray=Get Actors(self=Get DFX_Placement())
   )
   > Append Array(TargetArray=Get BuildingActors(), SourceArray=Get OverlappingActors(self=Get DFX_Placement()))
   > Return Node!(execute)
   n7 = OverlappingActors(execute=@n1.NewEnumerator1)
   > n8 = For Each Loop(Array=@n7)
   > Branch(Condition=IsBuildingOfAllowedClass(InActor=@n8))
   > Add Unselected Overlapping Actor(InActor=@n8)
  ]}
 Get DFX Actor {nodes [
   Get DFX Actor() {is=title K2Node_FunctionEntry}
   > n1 = GetSelectedActors()
   > n2 = For Each Loop(Array=@n1)
   > n3 = Cast To DFX_Placement(Object=@n2)
   n4 = Get DFX_Placement()
   > n5 = Is Valid(InputObject=@n4)
   > n6 = Branch(Condition="Not Equal (Object)"(A:Object Reference=@n3, B:Object Reference=@n4) -> (ReturnValue:bool=~) {is=title K2Node_PromotableOperator})
   Set DFX_Placement(execute=[@n6.then, @n5.Is Not Valid], DFX_Placement=@n3)
   > Set Initilize DFX UI Settings(NewObject=Get DFX_Placement())
   > Clear(TargetArray=Get BuildingActorsBuffered() -> (BuildingActorsBuffered:Array of Actor Object References))
   > Set MaterialSlotBuffer(MaterialSlotBuffer=Get MaterialSlot(self=Get DFX_Placement()) -> (MaterialSlot=0)) -> (Output_Get=0)
  ]}
 Get Meshes Materials {nodes [
   Get Meshes Materials() {is=title K2Node_FunctionEntry}
   > n1 = For Each Loop(Array=Get BuildingActors())
   n2 = Get StaticMesh(self=GetComponentByClass(self=@n1, ComponentClass=&StaticMeshComponent))
   > n3 = For Each Loop(Array:Array of Static Material Structures=Get StaticMaterials!(self:Static Mesh Object Reference=@n2) -> (StaticMaterials:Array of Static Material Structures)) -> (Array Element:Static Material Structure)
   Add(
    execute=@n3.Completed
    TargetMap:Map of Static Mesh Object References to DFX Mesh Materials Structures=Get UniqueStaticMeshes()
    Key=@n2
    Value:"DFX Mesh Materials Structure (by ref)"=Make DFX Mesh Materials!(BaseMaterial_6_#1:Array of Material Interface Object References=Get MaterialsArrayTemp() -> (MaterialsArrayTemp:Array of Material Interface Object References)) -> (DFX_MeshMaterials:DFX Mesh Materials Structure) {is=title K2Node_MakeStruct}
   )
   > Clear(TargetArray:Array of Material Interface Object References=Get MaterialsArrayTemp())
   n4 = Break Static Material!(StaticMaterial:"Static Material Structure (by ref)"=@n3) -> (MaterialInterface:Material Interface Object Reference, OverlayMaterialInterface:Material Interface Object Reference) {is=title K2Node_BreakStruct}
   n5 = Cast To MaterialInstanceConstant(execute=@n3.LoopBody, Object=@n4)
   > Add Unique(TargetArray=Get UniqueMaterialConstants(), NewItem=@n5)
   > Add(
    TargetArray:Array of Material Interface Object References=Get MaterialsArrayTemp()
    NewItem:"Material Interface Object Reference (by ref)"=@n5
   ) {is=title K2Node_CallArrayFunction}
   n6 = Cast To Material(execute=@n5.CastFailed, Object=@n4)
   > Add Unique(TargetArray=Get UniqueMaterialConstants(), NewItem=@n6)
   > Add(TargetArray=Get MaterialsArrayTemp(), NewItem=@n6)
   n7 = Cast To MaterialInstanceDynamic(execute=@n6.CastFailed, Object=@n4)
   n8 = Get Parent(self=@n7)
   > Add Unique(TargetArray=Get UniqueMaterialConstants(), NewItem=@n8)
   > Add(TargetArray=Get MaterialsArrayTemp(), NewItem=@n8)
   Get StaticMeshComponent()
  ]}
 Get Unique Materials {nodes [
   Get Unique Materials() {is=title K2Node_FunctionEntry}
   > n1 = For Each Loop(Array:Array of Actor Object References=Get BuildingActors()) -> (Array Element:Actor Object Reference)
   > Add Unique(
    TargetArray=Get UniqueMaterials()
    NewItem=GetMaterial!(
     self:Primitive Component Object Reference=GetComponentByClass(self=@n1, ComponentClass=&StaticMeshComponent)
     ElementIndex:int=0
    ) -> (ReturnValue)
   )
  ]}
 GetText {nodes [
   GetText() {is=title K2Node_FunctionEntry}
   > Set Text!(execute, Text="LOCTEXT(\"{0}{1}{2}\")") -> (then, Output_Get)
   > n1 = For Each Loop(Array:Array of Material Interface Object References=Get Queued Materials To Create()) -> (Array Element:Material Interface Object Reference)
   n2 = "To Text (String)"()
   n3 = Set Text(execute=@n1.Completed, Text=Format Text!(Format=Get Text!() -> (Text), 0:Text=@n2, 1:Text=@n2, 2:Text=@n2) -> (Result))
   > SetVisibility(
    self=Get CreateMissingMaterialButton() -> (CreateMissingMaterialButton:Button Object Reference)
    InVisibility=Select(Option 0:ESlateVisibility Enum=Visible, Option 1:ESlateVisibility Enum=Collapsed, Index=TextIsEmpty(InText=@n3)) -> (ReturnValue:ESlateVisibility Enum)
   )
   > Return Node!(execute, ReturnValue:Text=Get Text())
   n4 = GetBaseMaterial(execute=@n1.LoopBody, self=@n1)
   > n5 = GetStaticSwitchParameterSource(Material=@n4, ParameterName=Enable DirectFX)
   n6 = DFX Material Path(Mat Interface=@n1)
   > n7 = DoesAssetExist(self=Get EditorAssetSubsystem(), AssetPath=@n6)
   > n8 = Branch(Condition=@n7)
   n9 = "To Text (String)"(InString="\r\n{0}{1}{2}")
   > n10 = Set Text(Text=Format Text(
     Format=Get Text()
     0:String=MakeLiteralString(Value="-- No Nanite Material  --")
     1=Select(Option 0:Text="LOCTEXT(\" --Material Not Supported--\")", Option 1:Text="LOCTEXT(\" \")", Index=@n5.ReturnValue) -> (ReturnValue:Text)
     2=@n9
    ))
   SetIsEnabled(execute=[@n11.then, @n10.then], self=Get CreateMissingMaterialButton(), bInIsEnabled=@n5.ReturnValue)
   n11 = Set Text(
    execute=@n8.else
    Text=Format Text(
     Format=Get Text()
     0=DFX Material Path(Mat Interface=@n1).Name
     1=Select(Option 0="LOCTEXT(\" --Material Not Supported--\")", Index=@n5.ReturnValue)
     2=@n9
    )
   )
   GetPackageForObject(Object=@n1)
   GetAssetRegistry()
   String To Name(InString=@n6)
   Get MissingMaterials()
  ]}
 Get_MainMenu_Visibility {nodes [
   Get_MainMenu_Visibility() {is=title K2Node_FunctionEntry}
   n1 = Get DFX_Placement()
   > Return Node(ReturnValue:ESlateVisibility Enum=Select(Option 0:ESlateVisibility Enum=Collapsed, Option 1:ESlateVisibility Enum=SelfHitTestInvisible, Index=IsValid(Object=@n1)) -> (ReturnValue:ESlateVisibility Enum))
  ]}
 Get_Message_Visibility {nodes [
   Get_Message_Visibility() {is=title K2Node_FunctionEntry}
   n1 = Get DFX_Placement()
   > Return Node(ReturnValue=Select(Option 0=SelfHitTestInvisible, Option 1=Collapsed, Index=IsValid(Object=@n1)))
  ]}
 Get_ObjectName_Text {nodes [
   Get_ObjectName_Text() {is=title K2Node_FunctionEntry}
   n1 = Get DFX_Placement()
   > n2 = Is Valid(InputObject=@n1)
   Return Node(execute=[@n2.Is Not Valid, @n3.then], ReturnValue:Text=Get Text())
   > n3 = Set Text(Text="To Text (String)"(InString=GetDisplayName(Object=@n1)))
  ]}
 Get_RefreshBtn_Visibility {nodes [
   Get_RefreshBtn_Visibility() {is=title K2Node_FunctionEntry}
   n1 = Get DFX_Placement()
   > Return Node(ReturnValue:ESlateVisibility Enum=Select(Option 0=Collapsed, Option 1=Visible, Index=IsValid(Object=@n1)))
  ]}
 Get_TextOverlapping_Text {nodes [
   Get_TextOverlapping_Text() {is=title K2Node_FunctionEntry}
   n1 = Get DFX_Placement()
   > Is Valid(InputObject=@n1)
   n2 = Length(TargetArray=Get OverlappingActors(self=@n1))
   > Return Node(ReturnValue:Text=Select(Option 0:Text="LOCTEXT(\"0\")", Option 1:Text="To Text (Integer)"(Value=@n2), Index=IsValid(Object=@n1)) -> (ReturnValue:Text))
  ]}
 Get_TextSelected_Text {nodes [
   Get_TextSelected_Text() {is=title K2Node_FunctionEntry}
   n1 = Get DFX_Placement()
   > Is Valid(InputObject=@n1)
   n2 = Length(TargetArray=Get Actors(self=@n1))
   > Return Node(ReturnValue=Select(Option 0="LOCTEXT(\"0\")", Option 1="To Text (Integer)"(Value=@n2), Index=IsValid(Object=@n1)))
  ]}
 Get_TextTotal_Text {nodes [
   Get_TextTotal_Text() {is=title K2Node_FunctionEntry}
   > Is Valid(InputObject=Get DFX_Placement())
   n1 = Get DFX_Placement()
   n2 = int + int(A=Length(TargetArray=Get OverlappingActors(self=@n1)), B=Length(TargetArray=Get Actors(self=@n1)))
   > Return Node(ReturnValue=Select(Option 0="LOCTEXT(\"0\")", Option 1="To Text (Integer)"(Value=@n2), Index=IsValid(Object=@n1)))
  ]}
 IsBuildingOfAllowedClass {nodes [
   IsBuildingOfAllowedClass!() -> (then, InActor:Actor Object Reference) {is=title K2Node_FunctionEntry}
   > n1 = Sequence()
   > Cast To BuildingSMActor(Object=Get InActor() -> (InActor:Actor Object Reference))
   > Return Node!(execute, IsAllowedClass:bool=true)
   Return Node(execute=@n1.then_2, IsAllowedClass=false)
   Cast To BuildingStaticGeo(execute=@n1.then_1, Object=Get InActor())
   > Return Node(IsAllowedClass=true)
   Return Node(IsAllowedClass=true)
  ]}
 MakeMaterialStruct {nodes [
   Inputs!()
   n1 = Get DFX_Placement()
   n2 = GetTextures(self=@n1)
   n3 = Get ProjectionOverlap(self=Get DFX_Placement())
   n4 = "GetInverse (Matrix)"!(M:"Matrix Structure (by ref)"="To Matrix (Transform)"!(Transform:"Transform (by ref)"=Get World Transform!(self:Scene Component Object Reference=@n3) -> (ReturnValue:Transform) {is=title K2Node_CallFunction}) -> (ReturnValue:Matrix Structure)) -> (
    ReturnValue_XPlane_W:double=~
    ReturnValue_XPlane_X:double=~
    ReturnValue_XPlane_Y:double=~
    ReturnValue_XPlane_Z:double=~
    ReturnValue_YPlane_W:double=~
    ReturnValue_YPlane_X:double=~
    ReturnValue_YPlane_Y:double=~
    ReturnValue_YPlane_Z:double=~
    ReturnValue_ZPlane_W:double=~
    ReturnValue_ZPlane_X:double=~
    ReturnValue_ZPlane_Y:double=~
    ReturnValue_ZPlane_Z:double=~
    ReturnValue_WPlane:Plane Structure
   )
   n5 = Get Actor Location!(self=Get DFX_Placement()) -> (ReturnValue_X:double, ReturnValue_Y:double, ReturnValue_Z:double)
   n6 = GetRightVector!(self:Scene Component Object Reference=@n3) -> (ReturnValue_X:double, ReturnValue_Y:double, ReturnValue_Z:double)
   n7 = Get ProjectionFunction(self=@n1) -> (ProjectionFunction=NewEnumerator0)
   n8 = Get ProjectionFunction(self=Get DFX_Placement()) -> (ProjectionFunction=NewEnumerator0)
   n9 = Get DFX_Placement()
   n10 = Get AngleBlend!(self:DFX Placement Object Reference=@n9) -> (AngleBlend:double)
   n11 = GetForwardVector!(self:Scene Component Object Reference=@n3) -> (ReturnValue_X:double, ReturnValue_Y:double, ReturnValue_Z:double)
   n12 = Get World Scale!(self:Scene Component Object Reference=@n3) -> (ReturnValue_X:double, ReturnValue_Y:double, ReturnValue_Z:double) {is=title K2Node_CallFunction}
   n13 = float / float(A=@n12.ReturnValue_Z, B=@n12.ReturnValue_Y)
   Outputs!(MaterialParamsStruct:DFX Material Params Struct Structure=Make DFX Material Params Struct!(
     Texture_8_#2:Texture 2D Object Reference=Select!(
      NewEnumerator0:Texture 2D Object Reference=@n2
      NewEnumerator1:Texture 2D Object Reference=@n2.CausticTexture
      Index:Enum_DirectFX_Type Enum=@n7
     ) -> (ReturnValue:Texture 2D Object Reference)
     Matrix_X_9_#3:Linear Color Structure=MakeColor!(R:float=@n4.ReturnValue_XPlane_X, G:float=@n4.ReturnValue_XPlane_Y, B:float=@n4.ReturnValue_XPlane_Z, A:float=@n5) -> (ReturnValue:Linear Color Structure)
     Matrix_Y_10_#4:Linear Color Structure=MakeColor(R=@n4.ReturnValue_YPlane_X, G=@n4.ReturnValue_YPlane_Y, B=@n4.ReturnValue_YPlane_Z, A=@n5.ReturnValue_Y)
     Matrix_Z_11_#5:Linear Color Structure=MakeColor(R=@n4.ReturnValue_ZPlane_X, G=@n4.ReturnValue_ZPlane_Y, B=@n4.ReturnValue_ZPlane_Z, A=@n5.ReturnValue_Z)
     ForwardVector_13_#6:Linear Color Structure=MakeColor(R=@n11, G=@n11.ReturnValue_Y, B=@n11.ReturnValue_Z, A=@n13)
     TintColor_39_#7:Linear Color Structure=Get TintColor!(self:DFX Placement Object Reference=@n9) -> (TintColor:Linear Color Structure)
     PanningSpeed_42_#8:Linear Color Structure=Get PanningSpeed!(self:DFX Placement Object Reference=@n9) -> (PanningSpeed:Linear Color Structure)
     AngleBlend_18_#9:double=float * float(A=@n10, B=2) {is=title K2Node_PromotableOperator}
     AngleBlendHardness_19_#10:double=Get AngleBlendHardness!(self:DFX Placement Object Reference=@n9) -> (AngleBlendHardness:double)
     Perspective_22_#11:double=Get Perspective!(self:DFX Placement Object Reference=@n9) -> (Perspective:double)
     FixedTexture_25_#12:double=Select(NewEnumerator0:double=1, NewEnumerator1:double=0, Index=@n8) -> (ReturnValue:double)
     Tiling_27_#13:double=Get Tiling!(self:DFX Placement Object Reference=@n9) -> (Tiling:double)
     Intensity_30_#14:double=Get Intensity!(self:DFX Placement Object Reference=@n9) -> (Intensity:double)
     Falloff_35_#15:double=Get Falloff!(self:DFX Placement Object Reference=@n9) -> (Falloff:double)
     FalloffRadius_36_#16:double=Get FalloffRadius!(self:DFX Placement Object Reference=@n9) -> (FalloffRadius:double)
     Opacity_45_#17:double=Get Opacity!(self:DFX Placement Object Reference=@n9) -> (Opacity:double)
     NSRTexture_60_#18:Texture 2D Object Reference=@n2.NMRE Texture
     Specular_59_#19:double=Get Specular!(self:DFX Placement Object Reference=@n9) -> (Specular:double)
     Roughness_58_#20:double=Get Roughness!(self:DFX Placement Object Reference=@n9) -> (Roughness:double)
     Normal_57_#21:double=Get Normal!(self:DFX Placement Object Reference=@n9) -> (Normal:double)
     Emissive_56_#22:double=Get Emissive!(self:DFX Placement Object Reference=@n9) -> (Emissive:double)
     RightVector_64_#23:Linear Color Structure=Make Linear Color!(R:float=@n6, G:float=@n6.ReturnValue_Y, B:float=@n6.ReturnValue_Z, A:float=0) -> (LinearColor:Linear Color Structure) {is=title K2Node_MakeStruct}
     Metallic_67_#24:double=Get Metallic!(self:DFX Placement Object Reference=@n9) -> (Metallic:double)
     SpecularOverride_69_#25:double=Get SpecularOverride!(self:DFX Placement Object Reference=@n9) -> (SpecularOverride:double)
     TilingY_72_#26:double=Get TilingY!(self:DFX Placement Object Reference=@n9) -> (TilingY:double)
     TilingX_75_#27:double=Get TilingX!(self:DFX Placement Object Reference=@n9) -> (TilingX:double)
    ) -> (DFX_MaterialParamsStruct:DFX Material Params Struct Structure) {is=title K2Node_MakeStruct})
  ]}
 Manage DFX Components {nodes [
   Manage DFX Components() {is=title K2Node_FunctionEntry}
   > n1 = For Each Loop(Array:Array of Actor Object References=Get BuildingActors()) -> (Array Element:Actor Object Reference)
   n2 = For Each Loop(Exec=@n1.Completed, Array=Get BuildingActorsBuffered())
   Set BuildingActorsBuffered(execute=@n2.Completed, BuildingActorsBuffered:Array of Actor Object References=Get BuildingActors()) -> (Output_Get:Array of Actor Object References)
   n3 = GetComponentByClass(self=@n2, ComponentClass=&DFX_MaterialComponent_C) -> (ReturnValue:DFX Material Component Object Reference)
   Is Valid(exec=@n2.LoopBody, InputObject=@n3)
   > ResetMaterials(self=@n3)
   n4 = GetOwner(self=@n3)
   > n5 = Branch(Condition=IsBuildingOfAllowedClass(InActor=@n4) {is=title K2Node_CallFunction})
   > n6 = SetEditorProperty(Object=@n4, PropertyName=bAllowCustomMaterial, PropertyValue=Get bFalse() -> (bFalse:bool))
   Destroy DFX(execute=[@n5.else, @n6.then], self:DFX Material Component Object Reference=@n3)
   Is Valid(exec=@n1.LoopBody, InputObject=@n1)
   > Remove Item(TargetArray=Get BuildingActorsBuffered(), Item=@n1)
   n7 = GetComponentByClass(self=@n1, ComponentClass=&DFX_MaterialComponent_C)
   > n8 = Is Valid(InputObject=@n7)
   Add DFX Component!(execute=@n8.Is Not Valid, self, Actor:Actor Object Reference=@n1) -> (then) {is=title K2Node_CallFunction}
   n9 = GetComponentByClass(self=@n1, ComponentClass=&DFX_MaterialComponent_C)
   > Is Valid(InputObject=@n9)
   > Init(self:DFX Material Component Object Reference=@n9)
   > Add(
    TargetArray:Array of DFX Material Component Object References=Get DFXComponents()
    NewItem:"DFX Material Component Object Reference (by ref)"=@n9
   )
   n10 = Get DFX_Placement()
   Is Valid(exec=@n8.Is Valid, InputObject=@n10)
   > Set DecalPath!(execute, DecalPath:String=Get Object Path String(Object=@n10), self:DFX Material Component Object Reference=@n7) -> (then, Output_Get:String)
   > Add(TargetArray=Get DFXComponents(), NewItem=@n7)
  ]}
 NewDirectFX {nodes [
   NewDirectFX() {is=title K2Node_FunctionEntry}
   > n1 = GetSelectedActors()
   > n2 = Branch(Condition="integer > integer"(A=Length(TargetArray=@n1)))
   > n3 = GetSelectedActors()
   > n4 = Set LSelectedActors(LSelectedActors:Array of Actor Object References=@n3) -> (Output_Get:Array of Actor Object References)
   > n5 = For Each Loop(Array=@n4)
   n6 = Length(TargetArray=@n4)
   Set OriginsCombined(
    execute=@n5.Completed
    OriginsCombined:Vector=vector / vector(A:Vector=Get OriginsCombined() -> (OriginsCombined:Vector), B:int=@n6) -> (ReturnValue:Vector)
   ) -> (Output_Get:Vector)
   > Set ForwardsCombined(ForwardsCombined:Vector=vector / vector(A=Get ForwardsCombined() -> (ForwardsCombined:Vector), B=@n6)) -> (Output_Get:Vector)
   n7 = GetLevelViewportCameraInfo!(self:Unreal Editor Subsystem Object Reference=Get UnrealEditorSubsystem()) -> (CameraLocation:Vector, CameraRotation:Rotator, ReturnValue:bool)
   n8 = Normalize!(A:Vector=vector - vector(A=Get OriginsCombined(), B=@n7), Tolerance:float=0.0001) -> (ReturnValue:Vector) {is=title K2Node_CallFunction}
   > n9 = Multi Line Trace For Objects!(
    execute
    WorldContextObject:Object Reference
    Start:Vector=@n7
    End:Vector=vector + vector(A=Get OriginsCombined(), B=vector * vector(A=@n8, B:float=5000) -> (ReturnValue=~))
    ObjectTypes:Array of EObjectTypeQuery Enums=Get Trace Object Types() -> (Trace Object Types:Array of EObjectTypeQuery Enums=ObjectTypeQuery1)
    bTraceComplex:bool=false
    ActorsToIgnore:Array of Actor Object References
    DrawDebugType:EDrawDebugTrace Enum
    bIgnoreSelf:bool=true
    TraceColor:Linear Color Structure=(R=1,G=0,B=0,A=1)
    TraceHitColor:Linear Color Structure=(R=0,G=1,B=0,A=1)
    DrawTime:float=5
   ) -> (then, OutHits:Array of Hit Result Structures, ReturnValue:bool)
   > n10 = For Each Loop with Break(Array:Array of Hit Result Structures=@n9, Break=@n12.then_0) -> (Array Element:Hit Result Structure)
   n11 = BreakHitResult!(Hit:"Hit Result Structure (by ref)"=@n10) -> (
    bBlockingHit:bool
    bInitialOverlap:bool
    Time:float
    Distance:float
    Location:Vector
    ImpactPoint:Vector
    Normal:Vector
    ImpactNormal:Vector
    PhysMat:Physical Material Object Reference
    HitActor:Actor Object Reference
    HitComponent:Primitive Component Object Reference
    HitBoneName:Name
    BoneName:Name
    HitItem:int
    ElementIndex:int
    FaceIndex:int
    TraceStart:Vector
    TraceEnd:Vector
   )
   > Branch(Condition=Contains Item(
     TargetArray:Array of Actor Object References=Get LSelectedActors() -> (LSelectedActors:Array of Actor Object References)
     ItemToFind:"Actor Object Reference (by ref)"=@n11.HitActor
    ))
   > n12 = Sequence!(execute) -> (then_0, then_1)
   n13 = Get OriginsCombined()
   n14 = MakeRotFromX!(X:"Vector (by ref)"=vector * vector(A=@n11.ImpactNormal, B:double=-1) -> (ReturnValue=~)) -> (ReturnValue:Rotator)
   n15 = vector + vector(A=@n13, B=vector * vector(A:double=-10000, B:Vector=GetForwardVector!(InRot:Rotator=@n14) -> (ReturnValue:Vector)) -> (ReturnValue=~))
   n16 = Normalize(A=vector - vector(A=@n13, B=@n15), Tolerance=0.0001)
   n17 = Multi Line Trace For Objects(
    execute=@n12.then_1
    Start=@n15
    End=vector + vector(A=@n13, B=vector * vector(A:Vector=@n16, B:double=5000) -> (ReturnValue=~))
    ObjectTypes=Get Trace Object Types() -> (Trace Object Types=ObjectTypeQuery1)
    bTraceComplex=false
    bIgnoreSelf=true
    TraceColor=(R=1,G=0,B=0,A=1)
    TraceHitColor=(R=0,G=1,B=0,A=1)
    DrawTime=5
   )
   > n18 = For Each Loop with Break(Array=@n17, Break=@n20.then_0)
   n19 = BreakHitResult(Hit=@n18)
   > PrintString(InString=GetDisplayName(Object=@n19.HitActor))
   > Branch(Condition="Equal (Object)"(A=@n11.HitActor, B=@n19.HitActor))
   > n20 = Sequence()
   n21 = SpawnActorFromClass!(
    execute=@n20.then_1
    ActorClass:Actor Class Reference=&DFX_Placement_C
    Location:Vector=vector + vector(A=@n19.Location, B=vector * vector(A=@n19.Normal, B=70) -> (ReturnValue=~))
    Rotation:Rotator=@n14
    bTransient:bool
   ) -> (then, ReturnValue:DFX Placement Object Reference)
   > n22 = GetSelectedActors()
   > Set SelectedActors(SelectedActors=@n22)
   > SetSelectedLevelActors(self=Get EditorActorSubsystem(), ActorsToSelect=Make Array("[0]":Actor Object Reference=@n21) -> (Array:Array of Actor Object References))
   > Get DFX Actor() {is=title K2Node_CallFunction}
   > ForcePinSelectedActors() {is=title K2Node_CallFunction}
   Set OriginsCombined(execute=@n5.LoopBody, OriginsCombined=vector + vector(A=Get Actor Bounds(self=@n5, bOnlyCollidingComponents=false), B=Get OriginsCombined()))
   > Set ForwardsCombined(ForwardsCombined=vector + vector(A=GetActorForwardVector!(self:Actor Object Reference=@n5) -> (ReturnValue:Vector), B=Get ForwardsCombined()))
   n23 = GetLevelViewportCameraInfo(self=Get UnrealEditorSubsystem())
   n24 = vector + vector(A=@n23, B=vector * vector(A=GetForwardVector(InRot=@n23.CameraRotation), B:Vector="1000,1000,1000") -> (ReturnValue=~))
   n25 = Line Trace By Channel!(
    execute=@n2.else
    WorldContextObject:Object Reference
    Start:Vector=@n23
    End:Vector=@n24
    TraceChannel:ETraceTypeQuery Enum
    bTraceComplex:bool=false
    ActorsToIgnore:Array of Actor Object References
    DrawDebugType:EDrawDebugTrace Enum
    bIgnoreSelf:bool
    TraceColor:Linear Color Structure
    TraceHitColor:Linear Color Structure
    DrawTime:float
   ) -> (then, OutHit:Hit Result Structure, ReturnValue:bool)
   > n26 = SpawnActorFromClass(
    ActorClass=&DFX_Placement_C
    Location=SelectVector(A=BreakHitResult(Hit=@n25).Location, B=@n24, bPickA=@n25.ReturnValue)
   )
   > SetSelectedLevelActors(self=Get EditorActorSubsystem(), ActorsToSelect=Make Array("[0]"=@n26))
   > Get DFX Actor()
   > ForcePinSelectedActors()
  ]}
 OverlappingActors {nodes [
   OverlappingActors() {is=title K2Node_FunctionEntry}
   n1 = Get Box(self=Get DFX_Placement())
   > n2 = Component Overlap Actors!(
    execute
    Component:Primitive Component Object Reference=@n1
    ComponentTransform:"Transform (by ref)"=Get World Transform(self=@n1)
    ObjectTypes:Array of EObjectTypeQuery Enums=Make Array!("[0]":EObjectTypeQuery Enum=ObjectTypeQuery1, "[1]":EObjectTypeQuery Enum=ObjectTypeQuery2) -> (Array:Array of EObjectTypeQuery Enums)
    ActorClassFilter:Object Class Reference
    ActorsToIgnore:Array of Actor Object References
   ) -> (then, OutActors:Array of Actor Object References, ReturnValue:bool)
   > n3 = For Each Loop(Array=@n2)
   Return Node!(execute=@n3.Completed, Actors:Array of Actor Object References=Get OverlappingActors!() -> (OverlappingActors))
   > Branch(Condition=IsBuildingOfAllowedClass(InActor=@n3))
   n4 = Get Object Path String(Object=Get DFX_Placement())
   n5 = FindSubstring(SearchIn=@n4, Substring=:)
   n6 = Get Object Path String(Object=@n3)
   n7 = FindSubstring(SearchIn=@n6, Substring=:)
   > Branch(Condition="Equal Exactly (String)"(A=Left(SourceString=@n4, Count=@n5), B=Left(SourceString=@n6, Count=@n7)))
   > Add(TargetArray:Array of Actor Object References=Get OverlappingActors(), NewItem:"Actor Object Reference (by ref)"=@n3)
  ]}
 Set Initilize DFX UI Settings {nodes [
   n1 = Set Initilize DFX UI Settings!() -> (then, NewObject:Object Reference) {is=title K2Node_FunctionEntry}
   > SetObject(
    self=[
     Get SelectionMode!() -> (SelectionMode:Single Property View Object Reference)
     Get SortingInput() -> (SortingInput:Single Property View Object Reference)
     Get MaterialSlotInput() -> (MaterialSlotInput:Single Property View Object Reference)
     Get Type() -> (Type:Single Property View Object Reference)
    ]
    NewObject=@n1
   )
   n2 = Get DFX_DecalMenu()
   > SetObject(
    self=[
     Get DetailsView_41!(self:DFX Decal Menu Object Reference=@n2) -> (DetailsView_41:Details View Object Reference)
     Get DetailsView_62() -> (DetailsView_62:Details View Object Reference)
     Get DetailsViewNSR(self=@n2)
     Get EmissiveView(self=@n2)
     Get MetallicView(self=@n2)
     Get DecalTextureView!(self:DFX Decal Menu Object Reference=@n2) -> (DecalTextureView:Details View Object Reference)
     Get AlphaContrastView!(self:DFX Decal Menu Object Reference=@n2) -> (AlphaContrastView:Single Property View Object Reference)
     Get AlphaSourceView!(self:DFX Decal Menu Object Reference=@n2) -> (AlphaSourceView:Single Property View Object Reference)
     Get InvertAlphaView!(self:DFX Decal Menu Object Reference=@n2) -> (InvertAlphaView:Single Property View Object Reference)
     Get DetailsView_41(self:DFX Caustics Menu Object Reference=Get DFX_CausticsMenu() -> (DFX_CausticsMenu:DFX Caustics Menu Object Reference))
    ]
    NewObject=Get DFX_Placement()
   )
   n3 = Get DFX_Placement()
   > SetIsChecked(
    self=Get EditorUtilityCheckBox_1() -> (EditorUtilityCheckBox_1:Editor Utility Check Box Object Reference)
    InIsChecked=Get bVisible!(self:Scene Component Object Reference=Get ProjectionOverlap(self=@n3)) -> (bVisible:bool)
   )
   > SetObject(self=Get DetailsView_preview() -> (DetailsView_preview:Details View Object Reference), NewObject=Self-Reference())
   > UpdateNSRTextureStatus()
  ]}
 SetSubmenuVisibility {nodes [
   Inputs!() -> (execute:Exec)
   > SetVisibility(self=[Get DFX_CausticsMenu(), Get DFX_DecalMenu()], InVisibility=Collapsed)
   > n1 = Switch on Enum_DirectFX_Type!(execute, Selection:Enum_DirectFX_Type Enum=Get ProjectionFunction(self=Get DFX_Placement()) -> (ProjectionFunction=NewEnumerator0)) -> (NewEnumerator0:Exec, NewEnumerator1:Exec) {is=title K2Node_SwitchEnum}
   > n2 = SetVisibility(self=Get DFX_DecalMenu(), InVisibility=SelfHitTestInvisible)
   Outputs!(then:Exec=[@n2.then, @n3.then])
   n3 = SetVisibility(execute=@n1.NewEnumerator1, self=Get DFX_CausticsMenu(), InVisibility=SelfHitTestInvisible)
  ]}
 Update DFX Components {nodes [
   Update DFX Components() {is=title K2Node_FunctionEntry}
   > n1 = For Each Loop(Array:Array of DFX Material Component Object References=Get DFXComponents()) -> (Array Element:DFX Material Component Object Reference)
   > Is Valid(InputObject=@n1)
   > n2 = Set MaterialParamsStruct_Stored(MaterialParamsStruct_Stored=Make Material Struct())
   > SetEditorProperty(Object=@n1, PropertyName=DFX_MaterialStruct, PropertyValue:DFX Material Params Struct Structure=@n2)
   > Set AlphaContrast!(
    execute
    AlphaContrast:double=Get AlphaContrast!(self:DFX Placement Object Reference=Get DFX_Placement()) -> (AlphaContrast:double)
    self:DFX Material Component Object Reference=@n1
   ) -> (then, Output_Get:double)
   n3 = Get AlphaSource!(self:DFX Placement Object Reference=Get DFX_Placement()) -> (AlphaSource:E_DFX_AlphaSource Enum=NewEnumerator0)
   > n4 = Set AlphaSource!(execute, AlphaSource:E_DFX_AlphaSource Enum=@n3, self:DFX Material Component Object Reference=@n1) -> (then, Output_Get:E_DFX_AlphaSource Enum=NewEnumerator0)
   Set InvertAlpha!(
    execute=[@n5.then, @n4.then]
    InvertAlpha:bool=Get InvertAlpha!(self:DFX Placement Object Reference=Get DFX_Placement()) -> (InvertAlpha:bool)
    self:DFX Material Component Object Reference=@n1
   ) -> (then, Output_Get:bool)
   > Set MaterialSlot!(execute, MaterialSlot:int=Get MaterialSlot(self=Get DFX_Placement()) -> (MaterialSlot=0), self:DFX Material Component Object Reference=@n1) -> (then, Output_Get:int=0)
   > Update DFX(self:DFX Material Component Object Reference=@n1)
   n5 = SetEditorProperty(Object=@n1, PropertyName=AlphaSource, PropertyValue:E_DFX_AlphaSource Enum=@n3)
  ]}
 Widget Setup {nodes [
   Widget Setup() {is=title K2Node_FunctionEntry}
   > SetObject(self=Get DebugMaterial() -> (DebugMaterial:Single Property View Object Reference), NewObject=Self-Reference())
  ]}
]
variables [
 ActorsWithMissingComponent {type="array<BuildingSMActor>", default []}
 bFalse {type=boolean}
 bForceUpdate {type=boolean, default=true}
 bTrue {type=boolean}
 BuildingActors {type="array<Actor>"}
 BuildingActorsBuffered {type="array<Actor>"}
 BuildingActorsMeshes {default {}}
 DebugView {type=boolean}
 DFX_Placement {type=DFX_Placement_C, default=null}
 DFXActorLocationBuffer {type=Vector}
 DFXComponents {type="array<DFX_MaterialComponent_C>", default []}
 DFXMaterialTemp {type=Object, default=null}
 Init {type=boolean}
 MaterialParamsStruct {
  type=DFX_MaterialParamsStruct
  default {
   angleBlend=0.5
   angleBlendHardness=0.5
   emissive=0
   falloff=1
   falloffRadius=1
   fixedTexture=0
   intensity=1
   metallic=0
   nSR Texture=null
   normal=0
   opacity=0
   perspective=0
   roughness=0
   specular=0
   specularOverride=0
   texture=null
   tiling=1
   tilingX=0
   tilingY=0
   forwardVector=<rgba 1 0 0 0>
   matrix_X=<rgba 0 0 0 0>
   matrix_Y=<rgba 0 0 0 0>
   matrix_Z=<rgba 0 0 0 0>
   panningSpeed=<rgba 0 0 0 0>
   rightVector=<rgba 0 0 0 0>
   tintColor=<rgba 1 1 1 1>
  }
 }
 MaterialParamsStruct_Stored {
  type=DFX_MaterialParamsStruct
  default {
   angleBlend=0.5
   angleBlendHardness=0.5
   emissive=0
   falloff=1
   falloffRadius=1
   fixedTexture=0
   intensity=1
   metallic=0
   nSR Texture=null
   normal=0
   opacity=0
   perspective=0
   roughness=0
   specular=0
   specularOverride=0
   texture=null
   tiling=1
   tilingX=0
   tilingY=0
   forwardVector=<rgba 1 0 0 0>
   matrix_X=<rgba 0 0 0 0>
   matrix_Y=<rgba 0 0 0 0>
   matrix_Z=<rgba 0 0 0 0>
   panningSpeed=<rgba 0 0 0 0>
   rightVector=<rgba 0 0 0 0>
   tintColor=<rgba 1 1 1 1>
  }
 }
 MaterialSlotBuffer {type=integer}
 MICsPath {type=string, default=&/MICs/}
 ParentMaterialToUseForActor {type="array<MaterialInterface>", default []}
 Queued Materials To Create {type="array<MaterialInterface>", default []}
 SelectedActors {type="array<Actor>"}
 StaticGeoTData {default {}}
 SupportedMaterials {
  type="array<Material>"
  default [
   /Game/Athena/WinterArt/Materials/M_BaseSimple_4Layers
   /Game/Global/Base4Layers_2022/Material/P_FN_Base_2022
   /Game/Environments/Helios/Sets/SharkBait/Materials/P_FN_Base_2022_SharkBait
  ]
 }
 UniqueMaterialConstants {type="array<MaterialInterface>", default []}
 UniqueMaterials {type="array<MaterialInterface>", default []}
 UniqueStaticMeshes {default {}}
 UseEmissive {type=boolean}
]
widgets {
 info {rootWidgetClass=&CanvasPanel, widgetCount=69}
 widgets [
  CanvasPanel_0 {widgetClassPath=&CanvasPanel, widget=CanvasPanel_0}
  EditorUtilityScrollBox_98 {
   widgetClassPath=&EditorUtilityScrollBox
   parent=CanvasPanel_0
   widget=EditorUtilityScrollBox_98
   slot {class=CanvasPanelSlot, properties {layoutData {alignment=<0 0>, anchors {maximum=<1 1>, minimum=<0 0>}, offsets=<ltrb 0 0 0 0>}}}
  }
  MeshPanel {
   widgetClassPath=&CommonBorder
   bIsVariable=true
   parent=EditorUtilityScrollBox_98
   widget=MeshPanel
   properties {
    verticalAlignment=VAlign_Top
    background {
     bIsDynamicallyLoaded=false
     drawAs=Image
     imageType=NoImage
     mirroring=NoMirror
     resourceName=None
     resourceObject=&T_WhiteMask
     tiling=NoTile
     imageSize=<32 32>
     margin=<ltrb 0 0 0 0>
     outlineSettings {
      bUseBrushTransparency=false
      roundingType=HalfHeightRadius
      width=0
      color {colorUseRule=UseColor_Specified, specifiedColor=<rgba 0 0 0 0>}
      cornerRadii=<0 0 0 0>
     }
     tintColor {colorUseRule=UseColor_Specified, specifiedColor=<rgba 1 1 1 1>}
     uVRegion {bIsValid=false, max=<0 0>, min=<0 0>}
    }
    brushColor=<rgba 0.0103 0.0103 0.0103 1>
    padding=<ltrb 0 0 0 0>
   }
   slot {class=ScrollBoxSlot}
  }
  TopBox {widgetClassPath=&VerticalBox, parent=EditorUtilityScrollBox_98, widget=TopBox, slot {class=ScrollBoxSlot}}
  Header {
   widgetClassPath=&HorizontalBox
   parent=TopBox
   widget=Header
   slot {class=VerticalBoxSlot, properties {size {sizeRule=Fill, value=1}}}
  }
  Image_226 {
   widgetClassPath=&Image
   bIsVariable=true
   parent=Header
   widget=Image_226
   properties {brush {
     bIsDynamicallyLoaded=false
     drawAs=Image
     imageType=NoImage
     mirroring=NoMirror
     resourceName=None
     resourceObject=/Game/ArtTools/Tech_Art/DirectFX/Assets/Icons/DirectFXIcon
     tiling=NoTile
     imageSize=<16 16>
     margin=<ltrb 0 0 0 0>
     outlineSettings {
      bUseBrushTransparency=false
      roundingType=HalfHeightRadius
      width=0
      color {colorUseRule=UseColor_Specified, specifiedColor=<rgba 0 0 0 0>}
      cornerRadii=<0 0 0 0>
     }
     tintColor {colorUseRule=UseColor_Specified, specifiedColor=<rgba 1 1 1 1>}
     uVRegion {bIsValid=false, max=<0 0>, min=<0 0>}
    }}
   slot {class=HorizontalBoxSlot, properties {horizontalAlignment=HAlign_Center, verticalAlignment=VAlign_Center, padding=<ltrb 8 4 0 4>}}
  }
  ObjectName {
   widgetClassPath=&CommonTextBlock
   bIsVariable=true
   parent=Header
   widget=ObjectName
   properties {
    text=Decal Name
    font {
     bForceMonospaced=false
     bMaterialIsStencil=false
     fontMaterial=null
     fontObject=&Roboto
     letterSpacing=0
     monospacedWidth=1
     size=11
     skewAmount=0
     typefaceFontName=Regular
     outlineSettings {
      bApplyOutlineToDropShadows=false
      bMiteredCorners=false
      bSeparateFillAlpha=false
      outlineMaterial=null
      outlineSize=0
      outlineColor=<rgba 0 0 0 1>
     }
    }
    shadowOffset=<0 0>
   }
   slot {class=HorizontalBoxSlot, properties {verticalAlignment=VAlign_Center, padding=<ltrb 8 0 0 0>}}
  }
  HorizontalBox_2 {
   widgetClassPath=&HorizontalBox
   parent=Header
   widget=HorizontalBox_2
   slot {class=HorizontalBoxSlot, properties {horizontalAlignment=HAlign_Right, padding=<ltrb 0 9 8 9>, size {sizeRule=Fill, value=1}}}
  }
  AddBtn {
   widgetClassPath=&Button
   parent=HorizontalBox_2
   widget=AddBtn
   properties {widgetStyle {
     clickedSlateSound {resourceObject=null}
     disabled {
      bIsDynamicallyLoaded=false
      drawAs=RoundedBox
      imageType=NoImage
      mirroring=NoMirror
      resourceName=None
      resourceObject=null
      tiling=NoTile
      imageSize=<0 0>
      margin=<ltrb 0 0 0 0>
      outlineSettings {
       bUseBrushTransparency=false
       roundingType=FixedRadius
       width=1
       color {colorUseRule=UseColor_ColorTable, specifiedColor=<rgba 0 0 0 0>}
       cornerRadii=<4 4 4 4>
      }
      tintColor {colorUseRule=UseColor_Specified, specifiedColor=<rgba 0.0395 0.0395 0.0395 1>}
      uVRegion {bIsValid=false, max=<0 0>, min=<0 0>}
     }
     disabledForeground {colorUseRule=UseColor_ColorTable, specifiedColor=<rgba 0 0 0 0>}
     hovered {
      bIsDynamicallyLoaded=false
      drawAs=RoundedBox
      imageType=NoImage
      mirroring=NoMirror
      resourceName=None
      resourceObject=null
      tiling=NoTile
      imageSize=<32 32>
      margin=<ltrb 0 0 0 0>
      outlineSettings {
       bUseBrushTransparency=false
       roundingType=FixedRadius
       width=1
       color {colorUseRule=UseColor_ColorTable, specifiedColor=<rgba 0 0 0 0>}
       cornerRadii=<4 4 4 4>
      }
      tintColor {colorUseRule=UseColor_Specified, specifiedColor=<rgba 0.0953 0.0953 0.0953 1>}
      uVRegion {bIsValid=false, max=<0 0>, min=<0 0>}
     }
     hoveredForeground {colorUseRule=UseColor_ColorTable, specifiedColor=<rgba 0 0 0 0>}
     hoveredSlateSound {resourceObject=null}
     normal {
      bIsDynamicallyLoaded=false
      drawAs=RoundedBox
      imageType=NoImage
      mirroring=NoMirror
      resourceName=None
      resourceObject=null
      tiling=NoTile
      imageSize=<32 32>
      margin=<ltrb 0 0 0 0>
      outlineSettings {
       bUseBrushTransparency=false
       roundingType=FixedRadius
       width=1
       color {colorUseRule=UseColor_ColorTable, specifiedColor=<rgba 0 0 0 0>}
       cornerRadii=<4 4 4 4>
      }
      tintColor {colorUseRule=UseColor_Specified, specifiedColor=<rgba 0.0395 0.0395 0.0395 1>}
      uVRegion {bIsValid=false, max=<0 0>, min=<0 0>}
     }
     normalForeground {colorUseRule=UseColor_ColorTable, specifiedColor=<rgba 0 0 0 0>}
     normalPadding=<ltrb 12 1.5 12 1.5>
     pressed {
      bIsDynamicallyLoaded=false
      drawAs=RoundedBox
      imageType=NoImage
      mirroring=NoMirror
      resourceName=None
      resourceObject=null
      tiling=NoTile
      imageSize=<32 32>
      margin=<ltrb 0 0 0 0>
      outlineSettings {
       bUseBrushTransparency=false
       roundingType=FixedRadius
       width=1
       color {colorUseRule=UseColor_ColorTable, specifiedColor=<rgba 0 0 0 0>}
       cornerRadii=<4 4 4 4>
      }
      tintColor {colorUseRule=UseColor_Specified, specifiedColor=<rgba 0.0284 0.0284 0.0284 1>}
      uVRegion {bIsValid=false, max=<0 0>, min=<0 0>}
     }
     pressedForeground {colorUseRule=UseColor_ColorTable, specifiedColor=<rgba 0 0 0 0>}
     pressedPadding=<ltrb 12 2.5 12 0.5>
     pressedSlateSound {resourceObject=null}
    }}
   slot {class=HorizontalBoxSlot}
  }
  HorizontalBox_37 {
   widgetClassPath=&HorizontalBox
   parent=AddBtn
   widget=HorizontalBox_37
   slot {class=ButtonSlot, properties {padding=<ltrb 4 5 4 5>}}
  }
  Image_71 {
   widgetClassPath=&Image
   bIsVariable=true
   parent=HorizontalBox_37
   widget=Image_71
   properties {brush {
     bIsDynamicallyLoaded=false
     drawAs=Image
     imageType=NoImage
     mirroring=NoMirror
     resourceName=None
     resourceObject=&DirectFXIcon
     tiling=NoTile
     imageSize=<14 14>
     margin=<ltrb 0 0 0 0>
     outlineSettings {
      bUseBrushTransparency=false
      roundingType=HalfHeightRadius
      width=0
      color {colorUseRule=UseColor_Specified, specifiedColor=<rgba 0 0 0 0>}
      cornerRadii=<0 0 0 0>
     }
     tintColor {colorUseRule=UseColor_Specified, specifiedColor=<rgba 1 1 1 1>}
     uVRegion {bIsValid=false, max=<0 0>, min=<0 0>}
    }}
   slot {class=HorizontalBoxSlot, properties {horizontalAlignment=HAlign_Center, padding=<ltrb 0 0 3 0>}}
  }
  CommonTextBlock {
   widgetClassPath=&CommonTextBlock
   parent=HorizontalBox_37
   widget=CommonTextBlock
   properties {
    text=New
    font {
     bForceMonospaced=false
     bMaterialIsStencil=false
     fontMaterial=null
     fontObject=&Roboto
     letterSpacing=0
     monospacedWidth=1
     size=9
     skewAmount=0
     typefaceFontName=Regular
     outlineSettings {
      bApplyOutlineToDropShadows=false
      bMiteredCorners=false
      bSeparateFillAlpha=false
      outlineMaterial=null
      outlineSize=0
      outlineColor=<rgba 0 0 0 1>
     }
    }
    shadowOffset=<0 0>
   }
   slot {class=HorizontalBoxSlot, properties {horizontalAlignment=HAlign_Center, verticalAlignment=VAlign_Center}}
  }
  RefreshBtn {
   widgetClassPath=&Button
   parent=HorizontalBox_2
   widget=RefreshBtn
   properties {widgetStyle {
     clickedSlateSound {resourceObject=null}
     disabled {
      bIsDynamicallyLoaded=false
      drawAs=RoundedBox
      imageType=NoImage
      mirroring=NoMirror
      resourceName=None
      resourceObject=null
      tiling=NoTile
      imageSize=<0 0>
      margin=<ltrb 0 0 0 0>
      outlineSettings {
       bUseBrushTransparency=false
       roundingType=FixedRadius
       width=1
       color {colorUseRule=UseColor_ColorTable, specifiedColor=<rgba 0 0 0 0>}
       cornerRadii=<4 4 4 4>
      }
      tintColor {colorUseRule=UseColor_Specified, specifiedColor=<rgba 0.0395 0.0395 0.0395 1>}
      uVRegion {bIsValid=false, max=<0 0>, min=<0 0>}
     }
     disabledForeground {colorUseRule=UseColor_ColorTable, specifiedColor=<rgba 0 0 0 0>}
     hovered {
      bIsDynamicallyLoaded=false
      drawAs=RoundedBox
      imageType=NoImage
      mirroring=NoMirror
      resourceName=None
      resourceObject=null
      tiling=NoTile
      imageSize=<32 32>
      margin=<ltrb 0 0 0 0>
      outlineSettings {
       bUseBrushTransparency=false
       roundingType=FixedRadius
       width=1
       color {colorUseRule=UseColor_ColorTable, specifiedColor=<rgba 0 0 0 0>}
       cornerRadii=<4 4 4 4>
      }
      tintColor {colorUseRule=UseColor_Specified, specifiedColor=<rgba 0.0953 0.0953 0.0953 1>}
      uVRegion {bIsValid=false, max=<0 0>, min=<0 0>}
     }
     hoveredForeground {colorUseRule=UseColor_ColorTable, specifiedColor=<rgba 0 0 0 0>}
     hoveredSlateSound {resourceObject=null}
     normal {
      bIsDynamicallyLoaded=false
      drawAs=RoundedBox
      imageType=NoImage
      mirroring=NoMirror
      resourceName=None
      resourceObject=null
      tiling=NoTile
      imageSize=<32 32>
      margin=<ltrb 0 0 0 0>
      outlineSettings {
       bUseBrushTransparency=false
       roundingType=FixedRadius
       width=1
       color {colorUseRule=UseColor_ColorTable, specifiedColor=<rgba 0 0 0 0>}
       cornerRadii=<4 4 4 4>
      }
      tintColor {colorUseRule=UseColor_Specified, specifiedColor=<rgba 0.0395 0.0395 0.0395 1>}
      uVRegion {bIsValid=false, max=<0 0>, min=<0 0>}
     }
     normalForeground {colorUseRule=UseColor_ColorTable, specifiedColor=<rgba 0 0 0 0>}
     normalPadding=<ltrb 12 1.5 12 1.5>
     pressed {
      bIsDynamicallyLoaded=false
      drawAs=RoundedBox
      imageType=NoImage
      mirroring=NoMirror
      resourceName=None
      resourceObject=null
      tiling=NoTile
      imageSize=<32 32>
      margin=<ltrb 0 0 0 0>
      outlineSettings {
       bUseBrushTransparency=false
       roundingType=FixedRadius
       width=1
       color {colorUseRule=UseColor_ColorTable, specifiedColor=<rgba 0 0 0 0>}
       cornerRadii=<4 4 4 4>
      }
      tintColor {colorUseRule=UseColor_Specified, specifiedColor=<rgba 0.0284 0.0284 0.0284 1>}
      uVRegion {bIsValid=false, max=<0 0>, min=<0 0>}
     }
     pressedForeground {colorUseRule=UseColor_ColorTable, specifiedColor=<rgba 0 0 0 0>}
     pressedPadding=<ltrb 12 2.5 12 0.5>
     pressedSlateSound {resourceObject=null}
    }}
   slot {class=HorizontalBoxSlot}
  }
  HorizontalBox_3 {
   widgetClassPath=&HorizontalBox
   parent=RefreshBtn
   widget=HorizontalBox_3
   slot {class=ButtonSlot, properties {padding=<ltrb 4 5 4 5>}}
  }
  Image {
   widgetClassPath=&Image
   bIsVariable=true
   parent=HorizontalBox_3
   widget=Image
   properties {brush {
     bIsDynamicallyLoaded=false
     drawAs=Image
     imageType=NoImage
     mirroring=NoMirror
     resourceName=None
     resourceObject=&DirectFXIcon
     tiling=NoTile
     imageSize=<14 14>
     margin=<ltrb 0 0 0 0>
     outlineSettings {
      bUseBrushTransparency=false
      roundingType=HalfHeightRadius
      width=0
      color {colorUseRule=UseColor_Specified, specifiedColor=<rgba 0 0 0 0>}
      cornerRadii=<0 0 0 0>
     }
     tintColor {colorUseRule=UseColor_Specified, specifiedColor=<rgba 1 1 1 1>}
     uVRegion {bIsValid=false, max=<0 0>, min=<0 0>}
    }}
   slot {class=HorizontalBoxSlot, properties {horizontalAlignment=HAlign_Center, padding=<ltrb 0 0 3 0>}}
  }
  CommonTextBlock_13 {
   widgetClassPath=&CommonTextBlock
   parent=HorizontalBox_3
   widget=CommonTextBlock_13
   properties {
    text=Refresh
    font {
     bForceMonospaced=false
     bMaterialIsStencil=false
     fontMaterial=null
     fontObject=&Roboto
     letterSpacing=0
     monospacedWidth=1
     size=9
     skewAmount=0
     typefaceFontName=Regular
     outlineSettings {
      bApplyOutlineToDropShadows=false
      bMiteredCorners=false
      bSeparateFillAlpha=false
      outlineMaterial=null
      outlineSize=0
      outlineColor=<rgba 0 0 0 1>
     }
    }
    shadowOffset=<0 0>
   }
   slot {class=HorizontalBoxSlot, properties {horizontalAlignment=HAlign_Center, verticalAlignment=VAlign_Center}}
  }
  MainMenu {widgetClassPath=&VerticalBox, parent=EditorUtilityScrollBox_98, widget=MainMenu, slot {class=ScrollBoxSlot}}
  Instances {
   widgetClassPath=&HorizontalBox
   parent=MainMenu
   widget=Instances
   slot {class=VerticalBoxSlot, properties {verticalAlignment=VAlign_Top}}
  }
  VerticalBox {
   widgetClassPath=&VerticalBox
   parent=Instances
   widget=VerticalBox
   slot {class=HorizontalBoxSlot, properties {padding=<ltrb 8 8 8 8>, size {sizeRule=Fill, value=1}}}
  }
  DFXType {
   widgetClassPath=&HorizontalBox
   parent=VerticalBox
   widget=DFXType
   slot {class=VerticalBoxSlot, properties {padding=<ltrb 0 0 0 24>}}
  }
  Type {
   widgetClassPath=&SinglePropertyView
   parent=DFXType
   widget=Type
   properties {nameOverride=Type, propertyName=ProjectionFunction}
   slot {class=HorizontalBoxSlot, properties {size {sizeRule=Fill, value=1}}}
  }
  HorizontalBox_0 {
   widgetClassPath=&HorizontalBox
   parent=DFXType
   widget=HorizontalBox_0
   slot {class=HorizontalBoxSlot, properties {size {sizeRule=Fill, value=1}}}
  }
  CommonTextBlock_7 {
   widgetClassPath=&CommonTextBlock
   parent=HorizontalBox_0
   widget=CommonTextBlock_7
   properties {
    style=/Game/ArtTools/Font/TextStyle_RobotoEngineLike.TextStyle_RobotoEngineLike_C
    text=Gizmo
    colorAndOpacity {colorUseRule=UseColor_Specified, specifiedColor=<rgba 0.5 0.5 0.5 1>}
    font {
     bForceMonospaced=false
     bMaterialIsStencil=false
     fontMaterial=null
     fontObject=&Roboto
     letterSpacing=0
     monospacedWidth=1
     size=8
     skewAmount=0
     typefaceFontName=Regular
     outlineSettings {
      bApplyOutlineToDropShadows=false
      bMiteredCorners=false
      bSeparateFillAlpha=false
      outlineMaterial=null
      outlineSize=0
      outlineColor=<rgba 0 0 0 1>
     }
    }
    shadowOffset=<0 0>
   }
   slot {class=HorizontalBoxSlot, properties {verticalAlignment=VAlign_Center, padding=<ltrb 4 0 8 0>}}
  }
  EditorUtilityCheckBox_1 {
   widgetClassPath=&EditorUtilityCheckBox
   bIsVariable=true
   parent=HorizontalBox_0
   widget=EditorUtilityCheckBox_1
   properties {widgetStyle {
     checkBoxType=CheckBox
     backgroundHoveredImage {
      bIsDynamicallyLoaded=false
      drawAs=NoDrawType
      imageType=NoImage
      mirroring=NoMirror
      resourceName=None
      resourceObject=null
      tiling=NoTile
      imageSize=<18 18>
      margin=<ltrb 0 0 0 0>
      outlineSettings {
       bUseBrushTransparency=false
       roundingType=FixedRadius
       width=1
       color {colorUseRule=UseColor_ColorTable, specifiedColor=<rgba 0 0 0 0>}
       cornerRadii=<3 3 3 3>
      }
      tintColor {colorUseRule=UseColor_Specified, specifiedColor=<rgba 0.0048 0.0048 0.0048 1>}
      uVRegion {bIsValid=false, max=<0 0>, min=<0 0>}
     }
     backgroundImage {
      bIsDynamicallyLoaded=false
      drawAs=NoDrawType
      imageType=NoImage
      mirroring=NoMirror
      resourceName=None
      resourceObject=null
      tiling=NoTile
      imageSize=<18 18>
      margin=<ltrb 0 0 0 0>
      outlineSettings {
       bUseBrushTransparency=false
       roundingType=FixedRadius
       width=1
       color {colorUseRule=UseColor_ColorTable, specifiedColor=<rgba 0 0 0 0>}
       cornerRadii=<3 3 3 3>
      }
      tintColor {colorUseRule=UseColor_Specified, specifiedColor=<rgba 0.0048 0.0048 0.0048 0>}
      uVRegion {bIsValid=false, max=<0 0>, min=<0 0>}
     }
     backgroundPressedImage {
      bIsDynamicallyLoaded=false
      drawAs=NoDrawType
      imageType=NoImage
      mirroring=NoMirror
      resourceName=None
      resourceObject=null
      tiling=NoTile
      imageSize=<18 18>
      margin=<ltrb 0 0 0 0>
      outlineSettings {
       bUseBrushTransparency=false
       roundingType=FixedRadius
       width=1
       color {colorUseRule=UseColor_ColorTable, specifiedColor=<rgba 0 0 0 0>}
       cornerRadii=<3 3 3 3>
      }
      tintColor {colorUseRule=UseColor_Specified, specifiedColor=<rgba 0.0048 0.0048 0.0048 1>}
      uVRegion {bIsValid=false, max=<0 0>, min=<0 0>}
     }
     borderBackgroundColor {colorUseRule=UseColor_Specified, specifiedColor=<rgba 1 1 1 0>}
     checkedForeground {colorUseRule=UseColor_Specified, specifiedColor=<rgba 1 1 1 1>}
     checkedHoveredForeground {colorUseRule=UseColor_Specified, specifiedColor=<rgba 1 1 1 1>}
     checkedHoveredImage {
      bIsDynamicallyLoaded=false
      drawAs=Image
      imageType=NoImage
      mirroring=NoMirror
      resourceName=None
      resourceObject=/Game/ArtTools/Tech_Art/InstancedMeshesCluster/Icon/Icon_Visible
      tiling=NoTile
      imageSize=<0 0>
      margin=<ltrb 0 0 0 0>
      outlineSettings {
       bUseBrushTransparency=false
       roundingType=HalfHeightRadius
       width=0
       color {colorUseRule=UseColor_Specified, specifiedColor=<rgba 0 0 0 0>}
       cornerRadii=<0 0 0 0>
      }
      tintColor {colorUseRule=UseColor_Specified, specifiedColor=<rgba 1 1 1 1>}
      uVRegion {bIsValid=false, max=<0 0>, min=<0 0>}
     }
     checkedImage {
      bIsDynamicallyLoaded=false
      drawAs=Image
      imageType=NoImage
      mirroring=NoMirror
      resourceName=None
      resourceObject=&Icon_Visible
      tiling=NoTile
      imageSize=<0 0>
      margin=<ltrb 0 0 0 0>
      outlineSettings {
       bUseBrushTransparency=false
       roundingType=HalfHeightRadius
       width=0
       color {colorUseRule=UseColor_Specified, specifiedColor=<rgba 0 0 0 0>}
       cornerRadii=<0 0 0 0>
      }
      tintColor {colorUseRule=UseColor_Specified, specifiedColor=<rgba 1 1 1 1>}
      uVRegion {bIsValid=false, max=<0 0>, min=<0 0>}
     }
     checkedPressedForeground {colorUseRule=UseColor_Specified, specifiedColor=<rgba 1 1 1 1>}
     checkedPressedImage {
      bIsDynamicallyLoaded=false
      drawAs=Image
      imageType=NoImage
      mirroring=NoMirror
      resourceName=None
      resourceObject=&Icon_Visible
      tiling=NoTile
      imageSize=<0 0>
      margin=<ltrb 0 0 0 0>
      outlineSettings {
       bUseBrushTransparency=false
       roundingType=HalfHeightRadius
       width=0
       color {colorUseRule=UseColor_Specified, specifiedColor=<rgba 0 0 0 0>}
       cornerRadii=<0 0 0 0>
      }
      tintColor {colorUseRule=UseColor_Specified, specifiedColor=<rgba 1 1 1 1>}
      uVRegion {bIsValid=false, max=<0 0>, min=<0 0>}
     }
     checkedSlateSound {resourceObject=null}
     foregroundColor {colorUseRule=UseColor_Specified, specifiedColor=<rgba 1 1 1 1>}
     hoveredForeground {colorUseRule=UseColor_Specified, specifiedColor=<rgba 1 1 1 1>}
     hoveredSlateSound {resourceObject=null}
     padding=<ltrb 2 0 0 0>
     pressedForeground {colorUseRule=UseColor_Specified, specifiedColor=<rgba 1 1 1 1>}
     uncheckedHoveredImage {
      bIsDynamicallyLoaded=false
      drawAs=Image
      imageType=NoImage
      mirroring=NoMirror
      resourceName=None
      resourceObject=/Game/ArtTools/Tech_Art/InstancedMeshesCluster/Icon/Icon_Hidden
      tiling=NoTile
      imageSize=<0 0>
      margin=<ltrb 0 0 0 0>
      outlineSettings {
       bUseBrushTransparency=false
       roundingType=HalfHeightRadius
       width=0
       color {colorUseRule=UseColor_Specified, specifiedColor=<rgba 0 0 0 0>}
       cornerRadii=<0 0 0 0>
      }
      tintColor {colorUseRule=UseColor_Specified, specifiedColor=<rgba 1 1 1 1>}
      uVRegion {bIsValid=false, max=<0 0>, min=<0 0>}
     }
     uncheckedImage {
      bIsDynamicallyLoaded=false
      drawAs=Image
      imageType=NoImage
      mirroring=NoMirror
      resourceName=None
      resourceObject=&Icon_Hidden
      tiling=NoTile
      imageSize=<0 0>
      margin=<ltrb 0 0 0 0>
      outlineSettings {
       bUseBrushTransparency=false
       roundingType=HalfHeightRadius
       width=0
       color {colorUseRule=UseColor_Specified, specifiedColor=<rgba 0 0 0 0>}
       cornerRadii=<0 0 0 0>
      }
      tintColor {colorUseRule=UseColor_Specified, specifiedColor=<rgba 1 1 1 1>}
      uVRegion {bIsValid=false, max=<0 0>, min=<0 0>}
     }
     uncheckedPressedImage {
      bIsDynamicallyLoaded=false
      drawAs=Image
      imageType=NoImage
      mirroring=NoMirror
      resourceName=None
      resourceObject=&Icon_Hidden
      tiling=NoTile
      imageSize=<0 0>
      margin=<ltrb 0 0 0 0>
      outlineSettings {
       bUseBrushTransparency=false
       roundingType=HalfHeightRadius
       width=0
       color {colorUseRule=UseColor_Specified, specifiedColor=<rgba 0 0 0 0>}
       cornerRadii=<0 0 0 0>
      }
      tintColor {colorUseRule=UseColor_Specified, specifiedColor=<rgba 1 1 1 1>}
      uVRegion {bIsValid=false, max=<0 0>, min=<0 0>}
     }
     uncheckedSlateSound {resourceObject=null}
     undeterminedForeground {colorUseRule=UseColor_Specified, specifiedColor=<rgba 1 1 1 1>}
     undeterminedHoveredImage {
      bIsDynamicallyLoaded=false
      drawAs=Image
      imageType=NoImage
      mirroring=NoMirror
      resourceName=None
      resourceObject=&Icon_Visible
      tiling=NoTile
      imageSize=<0 0>
      margin=<ltrb 0 0 0 0>
      outlineSettings {
       bUseBrushTransparency=false
       roundingType=HalfHeightRadius
       width=0
       color {colorUseRule=UseColor_Specified, specifiedColor=<rgba 0 0 0 0>}
       cornerRadii=<0 0 0 0>
      }
      tintColor {colorUseRule=UseColor_Specified, specifiedColor=<rgba 1 1 1 1>}
      uVRegion {bIsValid=false, max=<0 0>, min=<0 0>}
     }
     undeterminedImage {
      bIsDynamicallyLoaded=false
      drawAs=Image
      imageType=NoImage
      mirroring=NoMirror
      resourceName=None
      resourceObject=&Icon_Visible
      tiling=NoTile
      imageSize=<0 0>
      margin=<ltrb 0 0 0 0>
      outlineSettings {
       bUseBrushTransparency=false
       roundingType=HalfHeightRadius
       width=0
       color {colorUseRule=UseColor_Specified, specifiedColor=<rgba 0 0 0 0>}
       cornerRadii=<0 0 0 0>
      }
      tintColor {colorUseRule=UseColor_Specified, specifiedColor=<rgba 1 1 1 1>}
      uVRegion {bIsValid=false, max=<0 0>, min=<0 0>}
     }
     undeterminedPressedImage {
      bIsDynamicallyLoaded=false
      drawAs=Image
      imageType=NoImage
      mirroring=NoMirror
      resourceName=None
      resourceObject=&Icon_Visible
      tiling=NoTile
      imageSize=<0 0>
      margin=<ltrb 0 0 0 0>
      outlineSettings {
       bUseBrushTransparency=false
       roundingType=HalfHeightRadius
       width=0
       color {colorUseRule=UseColor_Specified, specifiedColor=<rgba 0 0 0 0>}
       cornerRadii=<0 0 0 0>
      }
      tintColor {colorUseRule=UseColor_Specified, specifiedColor=<rgba 1 1 1 1>}
      uVRegion {bIsValid=false, max=<0 0>, min=<0 0>}
     }
    }}
   slot {class=HorizontalBoxSlot}
  }
  DFXType_1 {
   widgetClassPath=&HorizontalBox
   parent=VerticalBox
   widget=DFXType_1
   slot {class=VerticalBoxSlot, properties {padding=<ltrb 0 0 0 24>}}
  }
  DebugMaterial {
   widgetClassPath=&SinglePropertyView
   parent=DFXType_1
   widget=DebugMaterial
   properties {propertyName=DebugView}
   slot {class=HorizontalBoxSlot}
  }
  UseEmissiveCheckbox {
   widgetClassPath=&EditorUtilityCheckBox
   bIsVariable=true
   parent=DFXType_1
   widget=UseEmissiveCheckbox
   slot {class=HorizontalBoxSlot, properties {size {sizeRule=Fill, value=1}}}
  }
  CommonTextBlock_217 {
   widgetClassPath=&CommonTextBlock
   parent=UseEmissiveCheckbox
   widget=CommonTextBlock_217
   properties {
    style=&TextStyle_RobotoEngineLike_C
    text=Use Emissive
    colorAndOpacity {colorUseRule=UseColor_Specified, specifiedColor=<rgba 0.5 0.5 0.5 1>}
    font {
     bForceMonospaced=false
     bMaterialIsStencil=false
     fontMaterial=null
     fontObject=&Roboto
     letterSpacing=0
     monospacedWidth=1
     size=8
     skewAmount=0
     typefaceFontName=Regular
     outlineSettings {
      bApplyOutlineToDropShadows=false
      bMiteredCorners=false
      bSeparateFillAlpha=false
      outlineMaterial=null
      outlineSize=0
      outlineColor=<rgba 0 0 0 1>
     }
    }
    shadowOffset=<0 0>
   }
   slot {class=PanelSlot}
  }
  HorizontalBox {
   widgetClassPath=&HorizontalBox
   parent=DFXType_1
   widget=HorizontalBox
   slot {class=HorizontalBoxSlot, properties {size {sizeRule=Fill, value=1}}}
  }
  SelectionSettings {
   widgetClassPath=&HorizontalBox
   parent=VerticalBox
   widget=SelectionSettings
   slot {class=VerticalBoxSlot, properties {padding=<ltrb 0 0 0 8>}}
  }
  VerticalBox_2 {
   widgetClassPath=&VerticalBox
   parent=SelectionSettings
   widget=VerticalBox_2
   slot {class=HorizontalBoxSlot, properties {size {sizeRule=Fill, value=1}}}
  }
  SelectionMode {
   widgetClassPath=&SinglePropertyView
   parent=VerticalBox_2
   widget=SelectionMode
   properties {nameOverride=Selection Mode, propertyName=SelectionMode}
   slot {class=VerticalBoxSlot}
  }
  PinOverlappingActorsBtn {
   widgetClassPath=&Button
   parent=VerticalBox_2
   widget=PinOverlappingActorsBtn
   properties {widgetStyle {
     clickedSlateSound {resourceObject=null}
     disabled {
      bIsDynamicallyLoaded=false
      drawAs=RoundedBox
      imageType=NoImage
      mirroring=NoMirror
      resourceName=None
      resourceObject=null
      tiling=NoTile
      imageSize=<0 0>
      margin=<ltrb 0 0 0 0>
      outlineSettings {
       bUseBrushTransparency=false
       roundingType=FixedRadius
       width=1
       color {colorUseRule=UseColor_ColorTable, specifiedColor=<rgba 0 0 0 0>}
       cornerRadii=<4 4 4 4>
      }
      tintColor {colorUseRule=UseColor_Specified, specifiedColor=<rgba 0.0395 0.0395 0.0395 1>}
      uVRegion {bIsValid=false, max=<0 0>, min=<0 0>}
     }
     disabledForeground {colorUseRule=UseColor_ColorTable, specifiedColor=<rgba 0 0 0 0>}
     hovered {
      bIsDynamicallyLoaded=false
      drawAs=RoundedBox
      imageType=NoImage
      mirroring=NoMirror
      resourceName=None
      resourceObject=null
      tiling=NoTile
      imageSize=<32 32>
      margin=<ltrb 0 0 0 0>
      outlineSettings {
       bUseBrushTransparency=false
       roundingType=FixedRadius
       width=1
       color {colorUseRule=UseColor_ColorTable, specifiedColor=<rgba 0 0 0 0>}
       cornerRadii=<4 4 4 4>
      }
      tintColor {colorUseRule=UseColor_Specified, specifiedColor=<rgba 0.0953 0.0953 0.0953 1>}
      uVRegion {bIsValid=false, max=<0 0>, min=<0 0>}
     }
     hoveredForeground {colorUseRule=UseColor_ColorTable, specifiedColor=<rgba 0 0 0 0>}
     hoveredSlateSound {resourceObject=null}
     normal {
      bIsDynamicallyLoaded=false
      drawAs=RoundedBox
      imageType=NoImage
      mirroring=NoMirror
      resourceName=None
      resourceObject=null
      tiling=NoTile
      imageSize=<32 32>
      margin=<ltrb 0 0 0 0>
      outlineSettings {
       bUseBrushTransparency=false
       roundingType=FixedRadius
       width=1
       color {colorUseRule=UseColor_ColorTable, specifiedColor=<rgba 0 0 0 0>}
       cornerRadii=<4 4 4 4>
      }
      tintColor {colorUseRule=UseColor_Specified, specifiedColor=<rgba 0.0395 0.0395 0.0395 1>}
      uVRegion {bIsValid=false, max=<0 0>, min=<0 0>}
     }
     normalForeground {colorUseRule=UseColor_ColorTable, specifiedColor=<rgba 0 0 0 0>}
     normalPadding=<ltrb 12 1.5 12 1.5>
     pressed {
      bIsDynamicallyLoaded=false
      drawAs=RoundedBox
      imageType=NoImage
      mirroring=NoMirror
      resourceName=None
      resourceObject=null
      tiling=NoTile
      imageSize=<32 32>
      margin=<ltrb 0 0 0 0>
      outlineSettings {
       bUseBrushTransparency=false
       roundingType=FixedRadius
       width=1
       color {colorUseRule=UseColor_ColorTable, specifiedColor=<rgba 0 0 0 0>}
       cornerRadii=<4 4 4 4>
      }
      tintColor {colorUseRule=UseColor_Specified, specifiedColor=<rgba 0.0284 0.0284 0.0284 1>}
      uVRegion {bIsValid=false, max=<0 0>, min=<0 0>}
     }
     pressedForeground {colorUseRule=UseColor_ColorTable, specifiedColor=<rgba 0 0 0 0>}
     pressedPadding=<ltrb 12 2.5 12 0.5>
     pressedSlateSound {resourceObject=null}
    }}
   slot {class=VerticalBoxSlot, properties {horizontalAlignment=HAlign_Left, padding=<ltrb 0 4 0 0>}}
  }
  CommonTextBlock_6 {
   widgetClassPath=&CommonTextBlock
   parent=PinOverlappingActorsBtn
   widget=CommonTextBlock_6
   properties {
    text=Pin Overlapping Actors
    font {
     bForceMonospaced=false
     bMaterialIsStencil=false
     fontMaterial=null
     fontObject=&Roboto
     letterSpacing=0
     monospacedWidth=1
     size=9
     skewAmount=0
     typefaceFontName=Regular
     outlineSettings {
      bApplyOutlineToDropShadows=false
      bMiteredCorners=false
      bSeparateFillAlpha=false
      outlineMaterial=null
      outlineSize=0
      outlineColor=<rgba 0 0 0 1>
     }
    }
    shadowOffset=<0 0>
   }
   slot {class=ButtonSlot, properties {padding=<ltrb 0 4 0 4>}}
  }
  VerticalBox_3 {
   widgetClassPath=&VerticalBox
   parent=SelectionSettings
   widget=VerticalBox_3
   slot {class=HorizontalBoxSlot, properties {horizontalAlignment=HAlign_Left, verticalAlignment=VAlign_Top, size {sizeRule=Fill, value=0.5}}}
  }
  SortingInput {
   widgetClassPath=&SinglePropertyView
   parent=VerticalBox_3
   widget=SortingInput
   properties {nameOverride=Sort Priority, propertyName=SortPriority}
   slot {class=VerticalBoxSlot, properties {horizontalAlignment=HAlign_Left}}
  }
  MaterialSlotInput {
   widgetClassPath=&SinglePropertyView
   parent=VerticalBox_3
   widget=MaterialSlotInput
   properties {nameOverride=Material Slot, propertyName=MaterialSlot}
   slot {class=VerticalBoxSlot, properties {horizontalAlignment=HAlign_Left, padding=<ltrb 0 4 0 0>}}
  }
  ActorsCount {
   widgetClassPath=&HorizontalBox
   parent=VerticalBox
   widget=ActorsCount
   slot {class=VerticalBoxSlot, properties {padding=<ltrb 0 0 0 8>}}
  }
  CommonTextBlock_2 {
   widgetClassPath=&CommonTextBlock
   parent=ActorsCount
   widget=CommonTextBlock_2
   properties {
    style=&TextStyle_RobotoEngineLike_C
    text=Actors:
    colorAndOpacity {colorUseRule=UseColor_Specified, specifiedColor=<rgba 0.5 0.5 0.5 1>}
    font {
     bForceMonospaced=false
     bMaterialIsStencil=false
     fontMaterial=null
     fontObject=&Roboto
     letterSpacing=0
     monospacedWidth=1
     size=8
     skewAmount=0
     typefaceFontName=Regular
     outlineSettings {
      bApplyOutlineToDropShadows=false
      bMiteredCorners=false
      bSeparateFillAlpha=false
      outlineMaterial=null
      outlineSize=0
      outlineColor=<rgba 0 0 0 1>
     }
    }
    shadowOffset=<0 0>
   }
   slot {class=HorizontalBoxSlot, properties {verticalAlignment=VAlign_Center, padding=<ltrb 4 0 8 0>}}
  }
  HorizontalBox_146 {
   widgetClassPath=&HorizontalBox
   parent=ActorsCount
   widget=HorizontalBox_146
   slot {class=HorizontalBoxSlot, properties {verticalAlignment=VAlign_Center, padding=<ltrb 4 0 8 0>, size {sizeRule=Fill, value=1}}}
  }
  CommonTextBlock_3 {
   widgetClassPath=&CommonTextBlock
   parent=HorizontalBox_146
   widget=CommonTextBlock_3
   properties {
    style=&TextStyle_RobotoEngineLike_C
    text="Overlapping: "
    colorAndOpacity {colorUseRule=UseColor_Specified, specifiedColor=<rgba 0.5 0.5 0.5 1>}
    font {
     bForceMonospaced=false
     bMaterialIsStencil=false
     fontMaterial=null
     fontObject=&Roboto
     letterSpacing=0
     monospacedWidth=1
     size=8
     skewAmount=0
     typefaceFontName=Regular
     outlineSettings {
      bApplyOutlineToDropShadows=false
      bMiteredCorners=false
      bSeparateFillAlpha=false
      outlineMaterial=null
      outlineSize=0
      outlineColor=<rgba 0 0 0 1>
     }
    }
    shadowOffset=<0 0>
   }
   slot {class=HorizontalBoxSlot}
  }
  TextOverlapping {
   widgetClassPath=&CommonTextBlock
   bIsVariable=true
   parent=HorizontalBox_146
   widget=TextOverlapping
   properties {
    style=&TextStyle_RobotoEngineLike_C
    text=0
    colorAndOpacity {colorUseRule=UseColor_Specified, specifiedColor=<rgba 0.5 0.5 0.5 1>}
    font {
     bForceMonospaced=false
     bMaterialIsStencil=false
     fontMaterial=null
     fontObject=&Roboto
     letterSpacing=0
     monospacedWidth=1
     size=8
     skewAmount=0
     typefaceFontName=Regular
     outlineSettings {
      bApplyOutlineToDropShadows=false
      bMiteredCorners=false
      bSeparateFillAlpha=false
      outlineMaterial=null
      outlineSize=0
      outlineColor=<rgba 0 0 0 1>
     }
    }
    shadowOffset=<0 0>
   }
   slot {class=HorizontalBoxSlot}
  }
  HorizontalBox_145 {
   widgetClassPath=&HorizontalBox
   parent=ActorsCount
   widget=HorizontalBox_145
   slot {class=HorizontalBoxSlot, properties {verticalAlignment=VAlign_Center, padding=<ltrb 4 0 8 0>, size {sizeRule=Fill, value=1}}}
  }
  CommonTextBlock_4 {
   widgetClassPath=&CommonTextBlock
   parent=HorizontalBox_145
   widget=CommonTextBlock_4
   properties {
    style=&TextStyle_RobotoEngineLike_C
    text="Pinned: "
    colorAndOpacity {colorUseRule=UseColor_Specified, specifiedColor=<rgba 0.5 0.5 0.5 1>}
    font {
     bForceMonospaced=false
     bMaterialIsStencil=false
     fontMaterial=null
     fontObject=&Roboto
     letterSpacing=0
     monospacedWidth=1
     size=8
     skewAmount=0
     typefaceFontName=Regular
     outlineSettings {
      bApplyOutlineToDropShadows=false
      bMiteredCorners=false
      bSeparateFillAlpha=false
      outlineMaterial=null
      outlineSize=0
      outlineColor=<rgba 0 0 0 1>
     }
    }
    shadowOffset=<0 0>
   }
   slot {class=HorizontalBoxSlot}
  }
  TextSelected {
   widgetClassPath=&CommonTextBlock
   bIsVariable=true
   parent=HorizontalBox_145
   widget=TextSelected
   properties {
    style=&TextStyle_RobotoEngineLike_C
    text=0
    colorAndOpacity {colorUseRule=UseColor_Specified, specifiedColor=<rgba 0.5 0.5 0.5 1>}
    font {
     bForceMonospaced=false
     bMaterialIsStencil=false
     fontMaterial=null
     fontObject=&Roboto
     letterSpacing=0
     monospacedWidth=1
     size=8
     skewAmount=0
     typefaceFontName=Regular
     outlineSettings {
      bApplyOutlineToDropShadows=false
      bMiteredCorners=false
      bSeparateFillAlpha=false
      outlineMaterial=null
      outlineSize=0
      outlineColor=<rgba 0 0 0 1>
     }
    }
    shadowOffset=<0 0>
   }
   slot {class=HorizontalBoxSlot}
  }
  HorizontalBox_142 {
   widgetClassPath=&HorizontalBox
   parent=ActorsCount
   widget=HorizontalBox_142
   slot {class=HorizontalBoxSlot, properties {verticalAlignment=VAlign_Center, padding=<ltrb 4 0 8 0>, size {sizeRule=Fill, value=1}}}
  }
  CommonTextBlock_5 {
   widgetClassPath=&CommonTextBlock
   parent=HorizontalBox_142
   widget=CommonTextBlock_5
   properties {
    style=&TextStyle_RobotoEngineLike_C
    text="Total: "
    colorAndOpacity {colorUseRule=UseColor_Specified, specifiedColor=<rgba 0.5 0.5 0.5 1>}
    font {
     bForceMonospaced=false
     bMaterialIsStencil=false
     fontMaterial=null
     fontObject=&Roboto
     letterSpacing=0
     monospacedWidth=1
     size=8
     skewAmount=0
     typefaceFontName=Regular
     outlineSettings {
      bApplyOutlineToDropShadows=false
      bMiteredCorners=false
      bSeparateFillAlpha=false
      outlineMaterial=null
      outlineSize=0
      outlineColor=<rgba 0 0 0 1>
     }
    }
    shadowOffset=<0 0>
   }
   slot {class=HorizontalBoxSlot}
  }
  TextTotal {
   widgetClassPath=&CommonTextBlock
   bIsVariable=true
   parent=HorizontalBox_142
   widget=TextTotal
   properties {
    style=&TextStyle_RobotoEngineLike_C
    text=0
    colorAndOpacity {colorUseRule=UseColor_Specified, specifiedColor=<rgba 0.5 0.5 0.5 1>}
    font {
     bForceMonospaced=false
     bMaterialIsStencil=false
     fontMaterial=null
     fontObject=&Roboto
     letterSpacing=0
     monospacedWidth=1
     size=8
     skewAmount=0
     typefaceFontName=Regular
     outlineSettings {
      bApplyOutlineToDropShadows=false
      bMiteredCorners=false
      bSeparateFillAlpha=false
      outlineMaterial=null
      outlineSize=0
      outlineColor=<rgba 0 0 0 1>
     }
    }
    shadowOffset=<0 0>
   }
   slot {class=HorizontalBoxSlot}
  }
  HorizontalBox_1 {
   widgetClassPath=&HorizontalBox
   parent=VerticalBox
   widget=HorizontalBox_1
   slot {class=VerticalBoxSlot, properties {padding=<ltrb 0 4 0 0>, size {sizeRule=Fill, value=1}}}
  }
  PinSelectedActorsBtn {
   widgetClassPath=&Button
   parent=HorizontalBox_1
   widget=PinSelectedActorsBtn
   properties {widgetStyle {
     clickedSlateSound {resourceObject=null}
     disabled {
      bIsDynamicallyLoaded=false
      drawAs=RoundedBox
      imageType=NoImage
      mirroring=NoMirror
      resourceName=None
      resourceObject=null
      tiling=NoTile
      imageSize=<0 0>
      margin=<ltrb 0 0 0 0>
      outlineSettings {
       bUseBrushTransparency=false
       roundingType=FixedRadius
       width=1
       color {colorUseRule=UseColor_ColorTable, specifiedColor=<rgba 0 0 0 0>}
       cornerRadii=<4 4 4 4>
      }
      tintColor {colorUseRule=UseColor_Specified, specifiedColor=<rgba 0.0395 0.0395 0.0395 1>}
      uVRegion {bIsValid=false, max=<0 0>, min=<0 0>}
     }
     disabledForeground {colorUseRule=UseColor_ColorTable, specifiedColor=<rgba 0 0 0 0>}
     hovered {
      bIsDynamicallyLoaded=false
      drawAs=RoundedBox
      imageType=NoImage
      mirroring=NoMirror
      resourceName=None
      resourceObject=null
      tiling=NoTile
      imageSize=<32 32>
      margin=<ltrb 0 0 0 0>
      outlineSettings {
       bUseBrushTransparency=false
       roundingType=FixedRadius
       width=1
       color {colorUseRule=UseColor_ColorTable, specifiedColor=<rgba 0 0 0 0>}
       cornerRadii=<4 4 4 4>
      }
      tintColor {colorUseRule=UseColor_Specified, specifiedColor=<rgba 0.0953 0.0953 0.0953 1>}
      uVRegion {bIsValid=false, max=<0 0>, min=<0 0>}
     }
     hoveredForeground {colorUseRule=UseColor_ColorTable, specifiedColor=<rgba 0 0 0 0>}
     hoveredSlateSound {resourceObject=null}
     normal {
      bIsDynamicallyLoaded=false
      drawAs=RoundedBox
      imageType=NoImage
      mirroring=NoMirror
      resourceName=None
      resourceObject=null
      tiling=NoTile
      imageSize=<32 32>
      margin=<ltrb 0 0 0 0>
      outlineSettings {
       bUseBrushTransparency=false
       roundingType=FixedRadius
       width=1
       color {colorUseRule=UseColor_ColorTable, specifiedColor=<rgba 0 0 0 0>}
       cornerRadii=<4 4 4 4>
      }
      tintColor {colorUseRule=UseColor_Specified, specifiedColor=<rgba 0.0395 0.0395 0.0395 1>}
      uVRegion {bIsValid=false, max=<0 0>, min=<0 0>}
     }
     normalForeground {colorUseRule=UseColor_ColorTable, specifiedColor=<rgba 0 0 0 0>}
     normalPadding=<ltrb 12 1.5 12 1.5>
     pressed {
      bIsDynamicallyLoaded=false
      drawAs=RoundedBox
      imageType=NoImage
      mirroring=NoMirror
      resourceName=None
      resourceObject=null
      tiling=NoTile
      imageSize=<32 32>
      margin=<ltrb 0 0 0 0>
      outlineSettings {
       bUseBrushTransparency=false
       roundingType=FixedRadius
       width=1
       color {colorUseRule=UseColor_ColorTable, specifiedColor=<rgba 0 0 0 0>}
       cornerRadii=<4 4 4 4>
      }
      tintColor {colorUseRule=UseColor_Specified, specifiedColor=<rgba 0.0284 0.0284 0.0284 1>}
      uVRegion {bIsValid=false, max=<0 0>, min=<0 0>}
     }
     pressedForeground {colorUseRule=UseColor_ColorTable, specifiedColor=<rgba 0 0 0 0>}
     pressedPadding=<ltrb 12 2.5 12 0.5>
     pressedSlateSound {resourceObject=null}
    }}
   slot {class=HorizontalBoxSlot, properties {horizontalAlignment=HAlign_Center, verticalAlignment=VAlign_Center, size {sizeRule=Fill, value=1}}}
  }
  CommonTextBlock_9 {
   widgetClassPath=&CommonTextBlock
   parent=PinSelectedActorsBtn
   widget=CommonTextBlock_9
   properties {
    text=Pin Selected Actors
    font {
     bForceMonospaced=false
     bMaterialIsStencil=false
     fontMaterial=null
     fontObject=&Roboto
     letterSpacing=0
     monospacedWidth=1
     size=9
     skewAmount=0
     typefaceFontName=Regular
     outlineSettings {
      bApplyOutlineToDropShadows=false
      bMiteredCorners=false
      bSeparateFillAlpha=false
      outlineMaterial=null
      outlineSize=0
      outlineColor=<rgba 0 0 0 1>
     }
    }
    shadowOffset=<0 0>
   }
   slot {class=ButtonSlot, properties {padding=<ltrb 0 4 0 4>}}
  }
  UnpinSelectedActorsBtn {
   widgetClassPath=&Button
   parent=HorizontalBox_1
   widget=UnpinSelectedActorsBtn
   properties {widgetStyle {
     clickedSlateSound {resourceObject=null}
     disabled {
      bIsDynamicallyLoaded=false
      drawAs=RoundedBox
      imageType=NoImage
      mirroring=NoMirror
      resourceName=None
      resourceObject=null
      tiling=NoTile
      imageSize=<0 0>
      margin=<ltrb 0 0 0 0>
      outlineSettings {
       bUseBrushTransparency=false
       roundingType=FixedRadius
       width=1
       color {colorUseRule=UseColor_ColorTable, specifiedColor=<rgba 0 0 0 0>}
       cornerRadii=<4 4 4 4>
      }
      tintColor {colorUseRule=UseColor_Specified, specifiedColor=<rgba 0.0395 0.0395 0.0395 1>}
      uVRegion {bIsValid=false, max=<0 0>, min=<0 0>}
     }
     disabledForeground {colorUseRule=UseColor_ColorTable, specifiedColor=<rgba 0 0 0 0>}
     hovered {
      bIsDynamicallyLoaded=false
      drawAs=RoundedBox
      imageType=NoImage
      mirroring=NoMirror
      resourceName=None
      resourceObject=null
      tiling=NoTile
      imageSize=<32 32>
      margin=<ltrb 0 0 0 0>
      outlineSettings {
       bUseBrushTransparency=false
       roundingType=FixedRadius
       width=1
       color {colorUseRule=UseColor_ColorTable, specifiedColor=<rgba 0 0 0 0>}
       cornerRadii=<4 4 4 4>
      }
      tintColor {colorUseRule=UseColor_Specified, specifiedColor=<rgba 0.0953 0.0953 0.0953 1>}
      uVRegion {bIsValid=false, max=<0 0>, min=<0 0>}
     }
     hoveredForeground {colorUseRule=UseColor_ColorTable, specifiedColor=<rgba 0 0 0 0>}
     hoveredSlateSound {resourceObject=null}
     normal {
      bIsDynamicallyLoaded=false
      drawAs=RoundedBox
      imageType=NoImage
      mirroring=NoMirror
      resourceName=None
      resourceObject=null
      tiling=NoTile
      imageSize=<32 32>
      margin=<ltrb 0 0 0 0>
      outlineSettings {
       bUseBrushTransparency=false
       roundingType=FixedRadius
       width=1
       color {colorUseRule=UseColor_ColorTable, specifiedColor=<rgba 0 0 0 0>}
       cornerRadii=<4 4 4 4>
      }
      tintColor {colorUseRule=UseColor_Specified, specifiedColor=<rgba 0.0395 0.0395 0.0395 1>}
      uVRegion {bIsValid=false, max=<0 0>, min=<0 0>}
     }
     normalForeground {colorUseRule=UseColor_ColorTable, specifiedColor=<rgba 0 0 0 0>}
     normalPadding=<ltrb 12 1.5 12 1.5>
     pressed {
      bIsDynamicallyLoaded=false
      drawAs=RoundedBox
      imageType=NoImage
      mirroring=NoMirror
      resourceName=None
      resourceObject=null
      tiling=NoTile
      imageSize=<32 32>
      margin=<ltrb 0 0 0 0>
      outlineSettings {
       bUseBrushTransparency=false
       roundingType=FixedRadius
       width=1
       color {colorUseRule=UseColor_ColorTable, specifiedColor=<rgba 0 0 0 0>}
       cornerRadii=<4 4 4 4>
      }
      tintColor {colorUseRule=UseColor_Specified, specifiedColor=<rgba 0.0284 0.0284 0.0284 1>}
      uVRegion {bIsValid=false, max=<0 0>, min=<0 0>}
     }
     pressedForeground {colorUseRule=UseColor_ColorTable, specifiedColor=<rgba 0 0 0 0>}
     pressedPadding=<ltrb 12 2.5 12 0.5>
     pressedSlateSound {resourceObject=null}
    }}
   slot {class=HorizontalBoxSlot, properties {horizontalAlignment=HAlign_Center, verticalAlignment=VAlign_Center, size {sizeRule=Fill, value=1}}}
  }
  CommonTextBlock_11 {
   widgetClassPath=&CommonTextBlock
   parent=UnpinSelectedActorsBtn
   widget=CommonTextBlock_11
   properties {
    text=Unpin Selected Actors
    font {
     bForceMonospaced=false
     bMaterialIsStencil=false
     fontMaterial=null
     fontObject=&Roboto
     letterSpacing=0
     monospacedWidth=1
     size=9
     skewAmount=0
     typefaceFontName=Regular
     outlineSettings {
      bApplyOutlineToDropShadows=false
      bMiteredCorners=false
      bSeparateFillAlpha=false
      outlineMaterial=null
      outlineSize=0
      outlineColor=<rgba 0 0 0 1>
     }
    }
    shadowOffset=<0 0>
   }
   slot {class=ButtonSlot, properties {padding=<ltrb 0 4 0 4>}}
  }
  DetailsView_62 {
   widgetClassPath=&DetailsView
   parent=MainMenu
   widget=DetailsView_62
   properties {propertiesToShow [Actors]}
   slot {class=VerticalBoxSlot}
  }
  MissingMaterials {widgetClassPath=&HorizontalBox, bIsVariable=true, parent=MainMenu, widget=MissingMaterials, slot {class=VerticalBoxSlot}}
  VerticalBox_1 {
   widgetClassPath=&VerticalBox
   parent=MissingMaterials
   widget=VerticalBox_1
   slot {class=HorizontalBoxSlot, properties {padding=<ltrb 12 8 12 8>, size {sizeRule=Fill, value=1}}}
  }
  SnapSettings_2 {
   widgetClassPath=&HorizontalBox
   parent=VerticalBox_1
   widget=SnapSettings_2
   slot {class=VerticalBoxSlot, properties {padding=<ltrb 0 0 0 8>}}
  }
  CommonTextBlock_10 {
   widgetClassPath=&CommonTextBlock
   parent=SnapSettings_2
   widget=CommonTextBlock_10
   properties {
    style=&TextStyle_RobotoEngineLike_C
    text=Missing Materials:
    colorAndOpacity {colorUseRule=UseColor_Specified, specifiedColor=<rgba 0.5 0.5 0.5 1>}
    font {
     bForceMonospaced=false
     bMaterialIsStencil=false
     fontMaterial=null
     fontObject=&Roboto
     letterSpacing=0
     monospacedWidth=1
     size=8
     skewAmount=0
     typefaceFontName=Regular
     outlineSettings {
      bApplyOutlineToDropShadows=false
      bMiteredCorners=false
      bSeparateFillAlpha=false
      outlineMaterial=null
      outlineSize=0
      outlineColor=<rgba 0 0 0 1>
     }
    }
    shadowOffset=<0 0>
   }
   slot {class=HorizontalBoxSlot, properties {verticalAlignment=VAlign_Bottom, size {sizeRule=Fill, value=1}}}
  }
  CreateMissingMaterialButton {
   widgetClassPath=&Button
   parent=SnapSettings_2
   widget=CreateMissingMaterialButton
   properties {widgetStyle {
     clickedSlateSound {resourceObject=null}
     disabled {
      bIsDynamicallyLoaded=false
      drawAs=RoundedBox
      imageType=NoImage
      mirroring=NoMirror
      resourceName=None
      resourceObject=null
      tiling=NoTile
      imageSize=<0 0>
      margin=<ltrb 0 0 0 0>
      outlineSettings {
       bUseBrushTransparency=false
       roundingType=FixedRadius
       width=1
       color {colorUseRule=UseColor_ColorTable, specifiedColor=<rgba 0 0 0 0>}
       cornerRadii=<4 4 4 4>
      }
      tintColor {colorUseRule=UseColor_Specified, specifiedColor=<rgba 0.0395 0.0395 0.0395 1>}
      uVRegion {bIsValid=false, max=<0 0>, min=<0 0>}
     }
     disabledForeground {colorUseRule=UseColor_ColorTable, specifiedColor=<rgba 0 0 0 0>}
     hovered {
      bIsDynamicallyLoaded=false
      drawAs=RoundedBox
      imageType=NoImage
      mirroring=NoMirror
      resourceName=None
      resourceObject=null
      tiling=NoTile
      imageSize=<32 32>
      margin=<ltrb 0 0 0 0>
      outlineSettings {
       bUseBrushTransparency=false
       roundingType=FixedRadius
       width=1
       color {colorUseRule=UseColor_ColorTable, specifiedColor=<rgba 0 0 0 0>}
       cornerRadii=<4 4 4 4>
      }
      tintColor {colorUseRule=UseColor_Specified, specifiedColor=<rgba 0.0953 0.0953 0.0953 1>}
      uVRegion {bIsValid=false, max=<0 0>, min=<0 0>}
     }
     hoveredForeground {colorUseRule=UseColor_ColorTable, specifiedColor=<rgba 0 0 0 0>}
     hoveredSlateSound {resourceObject=null}
     normal {
      bIsDynamicallyLoaded=false
      drawAs=RoundedBox
      imageType=NoImage
      mirroring=NoMirror
      resourceName=None
      resourceObject=null
      tiling=NoTile
      imageSize=<32 32>
      margin=<ltrb 0 0 0 0>
      outlineSettings {
       bUseBrushTransparency=false
       roundingType=FixedRadius
       width=1
       color {colorUseRule=UseColor_ColorTable, specifiedColor=<rgba 0 0 0 0>}
       cornerRadii=<4 4 4 4>
      }
      tintColor {colorUseRule=UseColor_Specified, specifiedColor=<rgba 0.0395 0.0395 0.0395 1>}
      uVRegion {bIsValid=false, max=<0 0>, min=<0 0>}
     }
     normalForeground {colorUseRule=UseColor_ColorTable, specifiedColor=<rgba 0 0 0 0>}
     normalPadding=<ltrb 12 1.5 12 1.5>
     pressed {
      bIsDynamicallyLoaded=false
      drawAs=RoundedBox
      imageType=NoImage
      mirroring=NoMirror
      resourceName=None
      resourceObject=null
      tiling=NoTile
      imageSize=<32 32>
      margin=<ltrb 0 0 0 0>
      outlineSettings {
       bUseBrushTransparency=false
       roundingType=FixedRadius
       width=1
       color {colorUseRule=UseColor_ColorTable, specifiedColor=<rgba 0 0 0 0>}
       cornerRadii=<4 4 4 4>
      }
      tintColor {colorUseRule=UseColor_Specified, specifiedColor=<rgba 0.0284 0.0284 0.0284 1>}
      uVRegion {bIsValid=false, max=<0 0>, min=<0 0>}
     }
     pressedForeground {colorUseRule=UseColor_ColorTable, specifiedColor=<rgba 0 0 0 0>}
     pressedPadding=<ltrb 12 2.5 12 0.5>
     pressedSlateSound {resourceObject=null}
    }}
   slot {class=HorizontalBoxSlot, properties {horizontalAlignment=HAlign_Right, verticalAlignment=VAlign_Center}}
  }
  CommonTextBlock_1 {
   widgetClassPath=&CommonTextBlock
   parent=CreateMissingMaterialButton
   widget=CommonTextBlock_1
   properties {
    text=Create Missing Materials
    font {
     bForceMonospaced=false
     bMaterialIsStencil=false
     fontMaterial=null
     fontObject=&Roboto
     letterSpacing=0
     monospacedWidth=1
     size=9
     skewAmount=0
     typefaceFontName=Regular
     outlineSettings {
      bApplyOutlineToDropShadows=false
      bMiteredCorners=false
      bSeparateFillAlpha=false
      outlineMaterial=null
      outlineSize=0
      outlineColor=<rgba 0 0 0 1>
     }
    }
    shadowOffset=<0 0>
   }
   slot {class=ButtonSlot, properties {padding=<ltrb 0 4 0 4>}}
  }
  DirtyMaterialButton {
   widgetClassPath=&Button
   parent=SnapSettings_2
   widget=DirtyMaterialButton
   properties {widgetStyle {
     clickedSlateSound {resourceObject=null}
     disabled {
      bIsDynamicallyLoaded=false
      drawAs=RoundedBox
      imageType=NoImage
      mirroring=NoMirror
      resourceName=None
      resourceObject=null
      tiling=NoTile
      imageSize=<0 0>
      margin=<ltrb 0 0 0 0>
      outlineSettings {
       bUseBrushTransparency=false
       roundingType=FixedRadius
       width=1
       color {colorUseRule=UseColor_ColorTable, specifiedColor=<rgba 0 0 0 0>}
       cornerRadii=<4 4 4 4>
      }
      tintColor {colorUseRule=UseColor_Specified, specifiedColor=<rgba 0.0395 0.0395 0.0395 1>}
      uVRegion {bIsValid=false, max=<0 0>, min=<0 0>}
     }
     disabledForeground {colorUseRule=UseColor_ColorTable, specifiedColor=<rgba 0 0 0 0>}
     hovered {
      bIsDynamicallyLoaded=false
      drawAs=RoundedBox
      imageType=NoImage
      mirroring=NoMirror
      resourceName=None
      resourceObject=null
      tiling=NoTile
      imageSize=<32 32>
      margin=<ltrb 0 0 0 0>
      outlineSettings {
       bUseBrushTransparency=false
       roundingType=FixedRadius
       width=1
       color {colorUseRule=UseColor_ColorTable, specifiedColor=<rgba 0 0 0 0>}
       cornerRadii=<4 4 4 4>
      }
      tintColor {colorUseRule=UseColor_Specified, specifiedColor=<rgba 0.0953 0.0953 0.0953 1>}
      uVRegion {bIsValid=false, max=<0 0>, min=<0 0>}
     }
     hoveredForeground {colorUseRule=UseColor_ColorTable, specifiedColor=<rgba 0 0 0 0>}
     hoveredSlateSound {resourceObject=null}
     normal {
      bIsDynamicallyLoaded=false
      drawAs=RoundedBox
      imageType=NoImage
      mirroring=NoMirror
      resourceName=None
      resourceObject=null
      tiling=NoTile
      imageSize=<32 32>
      margin=<ltrb 0 0 0 0>
      outlineSettings {
       bUseBrushTransparency=false
       roundingType=FixedRadius
       width=1
       color {colorUseRule=UseColor_ColorTable, specifiedColor=<rgba 0 0 0 0>}
       cornerRadii=<4 4 4 4>
      }
      tintColor {colorUseRule=UseColor_Specified, specifiedColor=<rgba 0.0395 0.0395 0.0395 1>}
      uVRegion {bIsValid=false, max=<0 0>, min=<0 0>}
     }
     normalForeground {colorUseRule=UseColor_ColorTable, specifiedColor=<rgba 0 0 0 0>}
     normalPadding=<ltrb 12 1.5 12 1.5>
     pressed {
      bIsDynamicallyLoaded=false
      drawAs=RoundedBox
      imageType=NoImage
      mirroring=NoMirror
      resourceName=None
      resourceObject=null
      tiling=NoTile
      imageSize=<32 32>
      margin=<ltrb 0 0 0 0>
      outlineSettings {
       bUseBrushTransparency=false
       roundingType=FixedRadius
       width=1
       color {colorUseRule=UseColor_ColorTable, specifiedColor=<rgba 0 0 0 0>}
       cornerRadii=<4 4 4 4>
      }
      tintColor {colorUseRule=UseColor_Specified, specifiedColor=<rgba 0.0284 0.0284 0.0284 1>}
      uVRegion {bIsValid=false, max=<0 0>, min=<0 0>}
     }
     pressedForeground {colorUseRule=UseColor_ColorTable, specifiedColor=<rgba 0 0 0 0>}
     pressedPadding=<ltrb 12 2.5 12 0.5>
     pressedSlateSound {resourceObject=null}
    }}
   slot {class=HorizontalBoxSlot, properties {horizontalAlignment=HAlign_Right, verticalAlignment=VAlign_Center}}
  }
  CommonTextBlock_12 {
   widgetClassPath=&CommonTextBlock
   parent=DirtyMaterialButton
   widget=CommonTextBlock_12
   properties {
    text=Save Dirty Material
    font {
     bForceMonospaced=false
     bMaterialIsStencil=false
     fontMaterial=null
     fontObject=&Roboto
     letterSpacing=0
     monospacedWidth=1
     size=9
     skewAmount=0
     typefaceFontName=Regular
     outlineSettings {
      bApplyOutlineToDropShadows=false
      bMiteredCorners=false
      bSeparateFillAlpha=false
      outlineMaterial=null
      outlineSize=0
      outlineColor=<rgba 0 0 0 1>
     }
    }
    shadowOffset=<0 0>
   }
   slot {class=ButtonSlot, properties {padding=<ltrb 0 4 0 4>}}
  }
  CommonTextBlock_8 {
   widgetClassPath=&CommonTextBlock
   parent=VerticalBox_1
   widget=CommonTextBlock_8
   properties {
    text=Missing Materials:
    colorAndOpacity {colorUseRule=UseColor_Specified, specifiedColor=<rgba 1 0.5287 0 1>}
    font {
     bForceMonospaced=false
     bMaterialIsStencil=false
     fontMaterial=null
     fontObject=&Roboto
     letterSpacing=0
     monospacedWidth=1
     size=8
     skewAmount=0
     typefaceFontName=Regular
     outlineSettings {
      bApplyOutlineToDropShadows=false
      bMiteredCorners=false
      bSeparateFillAlpha=false
      outlineMaterial=null
      outlineSize=0
      outlineColor=<rgba 0 0 0 1>
     }
    }
    shadowOffset=<0 0>
   }
   slot {class=VerticalBoxSlot, properties {verticalAlignment=VAlign_Center}}
  }
  MaterialDirtyText {
   widgetClassPath=&CommonTextBlock
   bIsVariable=true
   parent=VerticalBox_1
   widget=MaterialDirtyText
   properties {
    text=Material Dirty
    colorAndOpacity {colorUseRule=UseColor_Specified, specifiedColor=<rgba 1 0.5287 0 1>}
    font {
     bForceMonospaced=false
     bMaterialIsStencil=false
     fontMaterial=null
     fontObject=&Roboto
     letterSpacing=0
     monospacedWidth=1
     size=8
     skewAmount=0
     typefaceFontName=Regular
     outlineSettings {
      bApplyOutlineToDropShadows=false
      bMiteredCorners=false
      bSeparateFillAlpha=false
      outlineMaterial=null
      outlineSize=0
      outlineColor=<rgba 0 0 0 1>
     }
    }
    shadowOffset=<0 0>
   }
   slot {class=VerticalBoxSlot, properties {verticalAlignment=VAlign_Center}}
  }
  DFX_CausticsMenu {
   widgetClassPath=/Game/ArtTools/Tech_Art/DirectFX/SubWidgets/DFX_CausticsMenu
   bIsVariable=true
   parent=MainMenu
   widget=DFX_CausticsMenu
   properties {visibility=Hidden}
   slot {class=VerticalBoxSlot, properties {verticalAlignment=VAlign_Top}}
  }
  DFX_DecalMenu {
   widgetClassPath=/Game/ArtTools/Tech_Art/DirectFX/SubWidgets/DFX_DecalMenu
   bIsVariable=true
   parent=MainMenu
   widget=DFX_DecalMenu
   properties {visibility=Hidden}
   slot {class=VerticalBoxSlot}
  }
  DetailsView_preview {
   widgetClassPath=&DetailsView
   parent=MainMenu
   widget=DetailsView_preview
   properties {visibility=Collapsed, propertiesToShow [UniqueStaticMeshes]}
   slot {class=VerticalBoxSlot}
  }
  SelectActorMessage {widgetClassPath=&VerticalBox, parent=EditorUtilityScrollBox_98, widget=SelectActorMessage, slot {class=ScrollBoxSlot}}
  SelectionMessage {
   widgetClassPath=&CommonTextBlock
   parent=SelectActorMessage
   widget=SelectionMessage
   properties {
    text=Select DFX Actor
    colorAndOpacity {colorUseRule=UseColor_Specified, specifiedColor=<rgba 0.276 0.276 0.276 1>}
    font {
     bForceMonospaced=false
     bMaterialIsStencil=false
     fontMaterial=null
     fontObject=&Roboto
     letterSpacing=0
     monospacedWidth=1
     size=8.25
     skewAmount=0
     typefaceFontName=Italic
     outlineSettings {
      bApplyOutlineToDropShadows=false
      bMiteredCorners=false
      bSeparateFillAlpha=false
      outlineMaterial=null
      outlineSize=0
      outlineColor=<rgba 0 0 0 1>
     }
    }
    shadowOffset=<0 0>
   }
   slot {
    class=VerticalBoxSlot
    properties {horizontalAlignment=HAlign_Center, verticalAlignment=VAlign_Top, padding=<ltrb 0 30 0 0>, size {sizeRule=Fill, value=1}}
   }
  }
 ]
}
