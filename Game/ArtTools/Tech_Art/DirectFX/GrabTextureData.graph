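# GrabTextureData is an editor utility Blueprint that reads building texture data from the selected actors in the editor.
# Its Construction Script calls Grab Data, which loops over the selection and breaks at the first successful cast to BuildingSMActor.
# It stores that actor, reads its BuildingTextureData array by property name through GetEditorProperty into NewVar, and prints whether the read succeeded.
# Nothing in the graph uses the stored array further.
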EditorUtilityBlueprint /Game/ArtTools/Tech_Art/DirectFX/GrabTextureData
graphs [
 EventGraph {nodes [
   Grab Data() {is=title K2Node_CustomEvent}
   > n1 = GetSelectedActors()
   > n2 = For Each Loop with Break!(Exec:Exec, Array:Array of Actor Object References=@n1, Break:Exec=@n4.then_0) -> (LoopBody:Exec, Array Element:Actor Object Reference, Array Index:int, Completed:Exec)
   > n3 = Cast To BuildingSMActor!(execute, Object=@n2) -> (then, CastFailed, AsBuilding Static Mesh:Building Static Mesh Object Reference) {is=title K2Node_DynamicCast}
   > n4 = Sequence()
   Set As Building Static Mesh(execute=@n4.then_1, As Building Static Mesh:Building Static Mesh Object Reference=@n3) -> (Output_Get:Building Static Mesh Object Reference)
   > n5 = GetEditorProperty!(execute:Exec, Object:Object Reference=@n3, PropertyName:Name=BuildingTextureData) -> (then:Exec, PropertyValue:Array of Building Texture Data Object References, ReturnValue:bool)
   > Set NewVar(NewVar:Array of Building Texture Data Object References=@n5) -> (Output_Get:Array of Building Texture Data Object References)
   > PrintString(InString="To String (Boolean)"(InBool=@n5.ReturnValue))
   Event BeginPlay()
   Event ActorBeginOverlap()
   Event Tick!() -> (OutputDelegate, then, DeltaSeconds:float)
   Get EditorActorSubsystem()
  ]}
 UserConstructionScript {nodes [
   Construction Script()
   > Grab Data() {is=title K2Node_CallFunction}
  ]}
]
variables [
 As Building Static Mesh {type=BuildingSMActor, default=null}
 NewVar {type="array<BuildingTextureData>", default []}
]
